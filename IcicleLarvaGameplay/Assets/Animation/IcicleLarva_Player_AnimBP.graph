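# Player animation layer Blueprint for IcicleLarva's flight ability, covering hover, normal, sprint and charge flight, plus ground slam, punch and heat vision states.
# ItemFinalPoseOverride runs a Primary state machine and caches per-mode state machines, each passed through its own linked adjustment layer.
# Adjustment layers stack additives, blendspaces and aim offsets; hover adds dynamic-additive pivots and root-bone offset.
# Conduit entries read "came from" flags, cleared by state functions, to pick transition animations.
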AnimBlueprint /IcicleLarvaGameplay/Assets/Animation/IcicleLarva_Player_AnimBP
target_skeleton=&Fortnite_M_Avg_Player_Skeleton
graphs [
 All Flight Adjustments {nodes [
   n1 = Layered blend per bone(
    BasePose=Additive Identity Pose()
    BlendPoses_0=Blendspace Player 'IcicleLarva_Flight_Pitch_BS'(X=0, Y=0) {blendSpace=/IcicleLarvaGameplay/Assets/Animation/Flight/HandsBackFlight/IcicleLarva_Flight_Pitch_BS}
   ) {layerSetup [{branchFilters [{boneName=Pelvis}, {blendDepth=-1, boneName=spine_01}, {blendDepth=-1, boneName=thigh_l}, {blendDepth=-1, boneName=thigh_r}]}]}
   n2 = Use cached pose 'AllFlightAdjustments Input Pose'()
   Output Pose(Result=Blend Poses by bool(
     BlendPose_0=Use cached pose 'AllFlightAdjustments Input Pose'()
     BlendPose_1=Apply Additive(Base=@n2, Additive=@n1) {alphaBoolBlend {blendOption=Linear}}
     BlendTime_0=0
     BlendTime_1=0
     bActiveValue=Get bIsGroundSlamMontage() -> (bIsGroundSlamMontage:bool)
    ) {blendType=Linear}) {layerGroup=Flight Adjustments}
   AimOffset Player 'M_NoWep_Relaxed_AO'!(
    X:float=0
    Y:float=Get!() -> (AdjustedAim_ResultingYaw:float=~, AdjustedAim_ResultingPitch:float=~, AdjustedAim_ZeroOutPitchWeightCurveName:Name).AdjustedAim_ResultingPitch
    BasePose:Pose Link Structure
    Alpha:float=1
   ) -> (Pose:Pose Link Structure) {blendSpace=/Game/Animation/Game/MainPlayer/Poses/Medium/Male/AO/M_NoWep_Relaxed_AO, alphaBoolBlend {blendOption=Linear}}
   Save cached pose 'AllFlightAdjustments Input Pose'(Pose=Input Pose() {name=InPose_1}) {cacheName=AllFlightAdjustments Input Pose}
   Get bIsNonHoverFlight() -> (bIsNonHoverFlight:bool)
  ]}
 AnimGraph {nodes [Output Pose()]}
 Charge Flight Adjustments {nodes [
   AnimGraphNode_StateMachine_0 = Charge Flight Adjustment SM() {is=title AnimGraphNode_StateMachine}
   Output Pose(Result=@AnimGraphNode_StateMachine_0) {layerGroup=Flight Adjustments}
   AnimGraphNode_StateMachine_3 = Settle() {is=title AnimGraphNode_StateMachine}
   Save cached pose 'Charge Settle Additive'(Pose=@AnimGraphNode_StateMachine_3) {cacheName=Charge Settle Additive}
   Save cached pose 'Charge Flight Adjustment Input Pose'(Pose=Input Pose() {name=ChargeInPose}) {cacheName=Charge Flight Adjustment Input Pose}
  ]}
 Charge Flight Adjustment SM {
  entry=Charge Flight Adjustment
  path=Charge Flight Adjustments.AnimGraphNode_StateMachine_0.Charge Flight Adjustment SM
  ids=id
  nodes [AnimStateNode_0 = Charge Flight Adjustment() {is=title AnimStateNode}]
 }
 Charge Flight Adjustment {
  path=Charge Flight Adjustments.AnimGraphNode_StateMachine_0.Charge Flight Adjustment SM.AnimStateNode_0.Charge Flight Adjustment
  nodes [
   n1 = "Sequence Player (additive) 'IcicleLarva_Flight_Sway_Idle_TwoHanded_Loop'"() {sequence=/IcicleLarvaGameplay/Assets/Animation/Flight/TwoHandedFlight/IcicleLarva_Flight_Sway_Idle_TwoHanded_Loop}
   n2 = Use cached pose 'Charge Flight Adjustment Input Pose'()
   n3 = Apply Additive(
    Base=Apply Additive(
     Base=Apply Additive(Base=@n2, Additive=@n1, Alpha=0.5) {alphaBoolBlend {blendOption=Linear}}
     Additive=Blendspace Player 'IcicleLarva_Flight_TwoHanded_BS_Old'(X=0, Y=0) {blendSpace=/IcicleLarvaGameplay/Assets/Animation/Flight/TwoHandedFlight/IcicleLarva_Flight_TwoHanded_BS_Old}
     Alpha=0.6
    ) {alphaBoolBlend {blendOption=Linear}}
    Additive=Blendspace Player 'IcicleLarva_Flight_Pitch_BS'(X=0, Y=0) {blendSpace=&IcicleLarva_Flight_Pitch_BS}
    Alpha=0
   ) {alphaBoolBlend {blendOption=Linear}}
   n4 = Blendspace Player 'IcicleLarva_Flight_TwoHanded_BS_Xtra'(X=0, Y=0) {blendSpace=/IcicleLarvaGameplay/Assets/Animation/Flight/TwoHandedFlight/IcicleLarva_Flight_TwoHanded_BS_Xtra}
   n5 = Apply Additive(
    Base=Apply Additive(
     Base=Apply Additive(Base=@n3, Additive=@n4, Alpha=0.25) {alphaBoolBlend {blendOption=Linear}}
     Additive=Blendspace Player 'IcicleLarva_Flight_TwoHanded_BS_Xtra'(X=0, Y=0) {blendSpace=&IcicleLarva_Flight_TwoHanded_BS_Xtra}
     Alpha=0
    ) {alphaBoolBlend {blendOption=Linear}}
    Additive=Use cached pose 'Charge Settle Additive'()
    Alpha=0
   ) {alphaBoolBlend {blendOption=Linear}}
   n6 = "Sequence Player (additive) 'IcicleLarva_Flight_Noise_Idle_TwoHanded_Loop'"() {sequence=/IcicleLarvaGameplay/Assets/Animation/Flight/TwoHandedFlight/IcicleLarva_Flight_Noise_Idle_TwoHanded_Loop}
   Output Animation Pose(Result=Apply Additive(
     Base=Apply Additive(Base=@n5, Additive=@n6) {alphaBoolBlend {blendOption=Linear}}
     Additive="Sequence Player (additive) 'IcicleLarva_Flight_Noise_Idle_TwoHanded_Loop'"() {sequence=&IcicleLarva_Flight_Noise_Idle_TwoHanded_Loop}
     Alpha=0
    ) {alphaBoolBlend {blendOption=Linear}})
  ]
 }
 Settle {
  entry=Passthrough
  path=Charge Flight Adjustments.AnimGraphNode_StateMachine_3.Settle
  ids=id
  nodes [
   Exit Settle() {is=title AnimStateAliasNode, also {aliases [Start Settle]}}
   AnimStateNode_0 = Passthrough() {is=title AnimStateNode}
   AnimStateNode_1 = Start Settle() {is=title AnimStateNode}
   AnimStateNode_2 = Settle() {is=title AnimStateNode}
   AnimStateTransitionNode_0 = AnimStateTransitionNode() {crossfadeDuration=0}
   AnimStateTransitionNode_2 = AnimStateTransitionNode() {crossfadeDuration=1}
   AnimStateTransitionNode_3 = AnimStateTransitionNode() {bAutomaticRuleBasedOnSequencePlayerInState=true}
   AnimStateTransitionNode_5 = AnimStateTransitionNode()
   AnimStateTransitionNode_6 = AnimStateTransitionNode()
  ]
  links [
   {
    dst=Passthrough
    rule=Charge Flight Adjustments.AnimGraphNode_StateMachine_3.Settle.AnimStateTransitionNode_5.Transition
    src=Exit Settle
    via=AnimStateTransitionNode_5
   }
   {
    dst=Passthrough
    rule=Charge Flight Adjustments.AnimGraphNode_StateMachine_3.Settle.AnimStateTransitionNode_6.Transition
    src=Exit Settle
    via=AnimStateTransitionNode_6
   }
   {
    dst=Start Settle
    rule=Charge Flight Adjustments.AnimGraphNode_StateMachine_3.Settle.AnimStateTransitionNode_0.Transition
    src=Passthrough
    via=AnimStateTransitionNode_0
   }
   {
    dst=Passthrough
    rule=Charge Flight Adjustments.AnimGraphNode_StateMachine_3.Settle.AnimStateTransitionNode_3.Transition
    src=Settle
    via=AnimStateTransitionNode_3
   }
   {
    dst=Settle
    rule=Charge Flight Adjustments.AnimGraphNode_StateMachine_3.Settle.AnimStateTransitionNode_2.Transition
    src=Start Settle
    via=AnimStateTransitionNode_2
   }
  ]
 }
 Passthrough {
  path=Charge Flight Adjustments.AnimGraphNode_StateMachine_3.Settle.AnimStateNode_0.Passthrough
  nodes [Output Animation Pose(Result=Additive Identity Pose())]
 }
 Start Settle {
  path=Charge Flight Adjustments.AnimGraphNode_StateMachine_3.Settle.AnimStateNode_1.Start Settle
  nodes [Output Animation Pose(Result=Additive Identity Pose()) {
    node {stateEntryFunction {functionName=OnStateEntry}}
    stateEntryFunction {bSelfContext=true, memberGuid=#1, memberName=OnStateEntry}
   }]
 }
 Settle {
  path=Charge Flight Adjustments.AnimGraphNode_StateMachine_3.Settle.AnimStateNode_2.Settle
  nodes [
   n1 = "Sequence Player (additive) 'IcicleLarva_Flight_Additive_BackLeft_Legs_Settle'"() {bLoopAnimation=false, sequence=/IcicleLarvaGameplay/Assets/Animation/Flight/HandsBackFlight/IcicleLarva_Flight_Additive_BackLeft_Legs_Settle}
   n2 = "Sequence Player (additive) 'IcicleLarva_Flight_Additive_BankRight_Legs_Settle'"() {
    bLoopAnimation=false
    sequence=/IcicleLarvaGameplay/Assets/Animation/Flight/HandsBackFlight/IcicleLarva_Flight_Additive_BankRight_Legs_Settle
   }
   Output Animation Pose(Result=Blend Poses by bool(BlendPose_0=@n1, BlendPose_1=@n2, bActiveValue=Get bSettleDirectionBlend() -> (bSettleDirectionBlend:bool)) {blendType=Linear})
  ]
 }
 Transition {
  path=Charge Flight Adjustments.AnimGraphNode_StateMachine_3.Settle.AnimStateTransitionNode_0.Transition
  nodes [Result(bCanEnterTransition=Get bPassThroughToStartSettleTransition() -> (bPassThroughToStartSettleTransition:bool))]
 }
 Transition {
  path=Charge Flight Adjustments.AnimGraphNode_StateMachine_3.Settle.AnimStateTransitionNode_2.Transition
  nodes [Result(bCanEnterTransition=true)]
 }
 Transition {path=Charge Flight Adjustments.AnimGraphNode_StateMachine_3.Settle.AnimStateTransitionNode_3.Transition, nodes [Result()]}
 Transition {
  path=Charge Flight Adjustments.AnimGraphNode_StateMachine_3.Settle.AnimStateTransitionNode_5.Transition
  nodes [Result(bCanEnterTransition=Get bExitSettleToPassThroughSecondTransition() -> (bExitSettleToPassThroughSecondTransition:bool))]
 }
 Transition {
  path=Charge Flight Adjustments.AnimGraphNode_StateMachine_3.Settle.AnimStateTransitionNode_6.Transition
  nodes [Result(bCanEnterTransition=Get bExitSettleToPassThroughTransition() -> (bExitSettleToPassThroughTransition:bool))]
 }
 EventGraph {links [], nodes []}
 FromChargeFalse {nodes [
   FromChargeFalse!() -> (then, Context:"Anim Update Context Structure (by ref)", Node:"Anim Node Reference Structure (by ref)") {is=title K2Node_FunctionEntry}
   > n1 = Branch(Condition=Get bIsAirCharging() -> (bIsAirCharging:bool))
   Set bIsFromCharge(execute=@n1.else, bIsFromCharge:bool=false) -> (Output_Get:bool)
   > Return Node()
  ]}
 FromChargeTrue {nodes [
   FromChargeTrue!() -> (then, Context:"Anim Update Context Structure (by ref)", Node:"Anim Node Reference Structure (by ref)") {is=title K2Node_FunctionEntry}
   > Set bIsFromCharge(bIsFromCharge=true)
   > Return Node()
  ]}
 FromFlightFalse {nodes [
   FromFlightFalse!() -> (then, Context:"Anim Update Context Structure (by ref)", Node:"Anim Node Reference Structure (by ref)") {is=title K2Node_FunctionEntry}
   > Set bIsHoverFlight(bIsHoverFlight:bool=false) -> (Output_Get:bool)
   > n1 = Branch(Condition=Get bIsNormalFlight() -> (bIsNormalFlight:bool))
   Set bIsFromFlight(execute=@n1.else, bIsFromFlight:bool=false) -> (Output_Get:bool)
   > Return Node()
  ]}
 FromFlightTrue {nodes [
   FromFlightTrue!() -> (then, Context:"Anim Update Context Structure (by ref)", Node:"Anim Node Reference Structure (by ref)") {is=title K2Node_FunctionEntry}
   > Set bIsFromFlight(bIsFromFlight=true)
   > Return Node()
  ]}
 FromHoverFalse {nodes [
   FromHoverFalse!() -> (then, Context:"Anim Update Context Structure (by ref)", Node:"Anim Node Reference Structure (by ref)") {is=title K2Node_FunctionEntry}
   > n1 = Branch(Condition=Get bIsHover() -> (bIsHover:bool))
   Set bIsFromHover(execute=@n1.else, bIsFromHover:bool=false) -> (Output_Get:bool)
   > Return Node()
  ]}
 FromHoverTrue {nodes [
   FromHoverTrue!() -> (then, Context:"Anim Update Context Structure (by ref)", Node:"Anim Node Reference Structure (by ref)") {is=title K2Node_FunctionEntry}
   > Set bIsFromHover(bIsFromHover=true)
   > Return Node()
  ]}
 FromIntroFalse {nodes [
   FromIntroFalse!() -> (then, Context:"Anim Update Context Structure (by ref)", Node:"Anim Node Reference Structure (by ref)") {is=title K2Node_FunctionEntry}
   > Set bIsFromIntro(bIsFromIntro:bool=false) -> (Output_Get:bool)
   > Return Node()
  ]}
 FromIntroTrue {nodes [
   FromIntroTrue!() -> (then, Context:"Anim Update Context Structure (by ref)", Node:"Anim Node Reference Structure (by ref)") {is=title K2Node_FunctionEntry}
   > Set bIsFromIntro(bIsFromIntro=true)
   > Return Node()
  ]}
 FromSprintFalse {nodes [
   FromSprintFalse!() -> (then, Context:"Anim Update Context Structure (by ref)", Node:"Anim Node Reference Structure (by ref)") {is=title K2Node_FunctionEntry}
   > n1 = Branch(Condition=Get bIsSprintFlight() -> (bIsSprintFlight:bool))
   Set bIsFromSprint(execute=@n1.else, bIsFromSprint:bool=false) -> (Output_Get:bool)
   > Return Node()
  ]}
 FromSprintTrue {nodes [
   FromSprintTrue!() -> (then, Context:"Anim Update Context Structure (by ref)", Node:"Anim Node Reference Structure (by ref)") {is=title K2Node_FunctionEntry}
   > Set bIsFromSprint(bIsFromSprint=true)
   > Return Node()
  ]}
 HeatVisionIntro {nodes [
   HeatVisionIntro!() -> (then, Context:"Anim Update Context Structure (by ref)", Node:"Anim Node Reference Structure (by ref)") {is=title K2Node_FunctionEntry}
   > Set bIsHeatVisionIntro(bIsHeatVisionIntro:bool=true) -> (Output_Get:bool)
   > Set bIsHeatVisionLoop(bIsHeatVisionLoop:bool=false) -> (Output_Get:bool)
   > Set bIsHeatVisionOutro(bIsHeatVisionOutro:bool=false) -> (Output_Get:bool)
   > Return Node()
  ]}
 HeatVisionLoop {nodes [
   HeatVisionLoop!() -> (then, Context:"Anim Update Context Structure (by ref)", Node:"Anim Node Reference Structure (by ref)") {is=title K2Node_FunctionEntry}
   > Set bIsHeatVisionIntro(bIsHeatVisionIntro=false)
   > Set bIsHeatVisionLoop(bIsHeatVisionLoop=true)
   > Set bIsHeatVisionOutro(bIsHeatVisionOutro=false)
   > Return Node()
  ]}
 HeatVisionOutro {nodes [
   HeatVisionOutro!() -> (then, Context:"Anim Update Context Structure (by ref)", Node:"Anim Node Reference Structure (by ref)") {is=title K2Node_FunctionEntry}
   > Set bIsHeatVisionIntro(bIsHeatVisionIntro=false)
   > Set bIsHeatVisionLoop(bIsHeatVisionLoop=false)
   > Set bIsHeatVisionOutro(bIsHeatVisionOutro=true)
   > Return Node()
  ]}
 Hover Adjustments {nodes [
   AnimGraphNode_StateMachine_1 = Hover Pivot() {is=title AnimGraphNode_StateMachine}
   n1 = Make Dynamic Additive!(Base:Pose Link Structure=Use cached pose 'Delta Pose'(), Additive:Pose Link Structure=@AnimGraphNode_StateMachine_1) -> (Pose:Pose Link Structure) {is=title AnimGraphNode_MakeDynamicAdditive}
   n2 = Use cached pose 'Hover Adjustment Cached Pose'()
   n3 = AimOffset Player 'IcicleLarva_Hover_AO_Xtra'!(
    X:float=0
    Y:float=0
    BasePose:Pose Link Structure=AimOffset Player 'IcicleLarva_Hover_AO'!(
     X:float=0
     Y:float=0
     BasePose:Pose Link Structure=Apply Additive(Base=@n2, Additive=@n1, Alpha=0.6) {alphaBoolBlend {blendOption=Linear}}
     Alpha:float=1
    ) -> (Pose:Pose Link Structure) {blendSpace=/IcicleLarvaGameplay/Assets/Animation/Flight/Hover/IcicleLarva_Hover_AO, alphaBoolBlend {blendOption=Linear}}
    Alpha:float=1
   ) -> (Pose:Pose Link Structure) {blendSpace=/IcicleLarvaGameplay/Assets/Animation/Flight/Hover/Xtras/IcicleLarva_Hover_AO_Xtra, alphaBoolBlend {blendOption=Linear}}
   n4 = Get RotationHalfLife() -> (RotationHalfLife:double)
   n5 = Get RotationMode() -> (RotationMode:EOffsetRootBoneMode Enum=Accumulate)
   n6 = Modify Curve!(
    SourcePose=Modify Curve(
     SourcePose=Offset Root Bone!(Source:Pose Link Structure=@n3, RotationMode:EOffsetRootBoneMode Enum=@n5, RotationHalfLife:float=@n4) -> (Pose:Pose Link Structure) {is=title AnimGraphNode_OffsetRootBone, tag=RootOffset, maxTranslationError=0, rotationMode=Accumulate}
     CurveValues_0=1
    ) {curveNames [DisableFlightPitchAdditive], curveValues [0]}
    CurveValues_0=1
   ) -> (Pose) {curveNames [DisableNoWeapAO], curveValues [0]}
   n7 = "Sequence Player (additive) 'IcicleLarva_Flight_Additive_Hover_Sway'"() {sequence=/IcicleLarvaGameplay/Assets/Animation/Flight/Hover/IcicleLarva_Flight_Additive_Hover_Sway}
   Output Pose(Result=Apply Additive(
     Base=Apply Additive(Base=@n6, Additive=@n7) {alphaBoolBlend {blendOption=Linear}}
     Additive="Sequence Player (additive) 'IcicleLarva_Flight_Additive_Hover_Sway'"() {sequence=&IcicleLarva_Flight_Additive_Hover_Sway}
    ) {alphaBoolBlend {blendOption=Linear}}) {layerGroup=Flight Adjustments}
   Save cached pose 'Hover Input Pose'(Pose=Input Pose()) {cacheName=Hover Input Pose}
   AnimGraphNode_StateMachine_0 = HoverAdjustments SM() {is=title AnimGraphNode_StateMachine}
   Save cached pose 'Hover Adjustment Cached Pose'(Pose=@AnimGraphNode_StateMachine_0) {cacheName=Hover Adjustment Cached Pose}
   Save cached pose 'Delta Pose'(Pose=Apply Additive(
     Base=Use cached pose 'Hover Input Pose'()
     Additive=Blendspace Player 'IcicleLarva_Hover_BS'(X=0, Y=0) {blendSpace=/IcicleLarvaGameplay/Assets/Animation/Flight/Hover/IcicleLarva_Hover_BS}
    ) {alphaBoolBlend {blendOption=Linear}}) {cacheName=Delta Pose}
  ]}
 HoverAdjustments SM {
  entry=Hover
  path=Hover Adjustments.AnimGraphNode_StateMachine_0.HoverAdjustments SM
  ids=id
  nodes [
   AnimStateNode_0 = Hover() {is=title AnimStateNode}
   AnimStateNode_1 = Hover Ascend() {is=title AnimStateNode}
   AnimStateTransitionNode_1 = AnimStateTransitionNode() {crossfadeDuration=0}
   AnimStateTransitionNode_2 = AnimStateTransitionNode() {crossfadeDuration=0.3}
  ]
  links [
   {
    dst=Hover Ascend
    rule=Hover Adjustments.AnimGraphNode_StateMachine_0.HoverAdjustments SM.AnimStateTransitionNode_1.Transition
    src=Hover
    via=AnimStateTransitionNode_1
   }
   {
    dst=Hover
    rule=Hover Adjustments.AnimGraphNode_StateMachine_0.HoverAdjustments SM.AnimStateTransitionNode_2.Transition
    src=Hover Ascend
    via=AnimStateTransitionNode_2
   }
  ]
 }
 Hover {
  path=Hover Adjustments.AnimGraphNode_StateMachine_0.HoverAdjustments SM.AnimStateNode_0.Hover
  nodes [
   Bone Mask!(
    BasePose:Pose Link Structure
    BlendPoses_0:Pose Link Structure
    BlendWeights_0:float=1
    BoneMask:Bone Mask Structure=(BoneMaskMap=())
    BoneMaskDefinitionDataAsset:Bone Mask Definition Data Asset Object Reference
   ) -> (Pose:Pose Link Structure) {is=title AnimGraphNode_BoneMask}
   n1 = Apply Additive(
    Base=Use cached pose 'Hover Input Pose'()
    Additive="Sequence Player (additive) 'IcicleLarva_Flight_Hovering_Idle_Up_And_Down_Xtra'"() {sequence=/IcicleLarvaGameplay/Assets/Animation/Flight/IcicleLarva_Flight_Hovering_Idle_Up_And_Down_Xtra}
   ) {alphaBoolBlend {blendOption=Linear}}
   n2 = Blendspace Player 'IcicleLarva_Hover_BS'(X=0, Y=0) {blendSpace=&IcicleLarva_Hover_BS}
   n3 = Layered blend per bone(
    BasePose=Additive Identity Pose()
    BlendPoses_0=Blendspace Player 'IcicleLarva_Hover_BS_Counter'(X=0, Y=0) {blendSpace=/IcicleLarvaGameplay/Assets/Animation/Flight/Hover/Xtras/IcicleLarva_Hover_BS_Counter}
   ) {
    bMeshSpaceRotationBlend=true
    blendMode=BlendMask
    blendMasks [&Fortnite_M_Avg_Player_Skeleton:LowerBody]
    layerSetup []
   }
   Output Animation Pose(Result=Apply Additive(Base=Apply Additive(Base=@n1, Additive=@n2) {alphaBoolBlend {blendOption=Linear}}, Additive=@n3) {alphaBoolBlend {blendOption=Linear}})
  ]
 }
 Hover Ascend {
  path=Hover Adjustments.AnimGraphNode_StateMachine_0.HoverAdjustments SM.AnimStateNode_1.Hover Ascend
  nodes [
   n1 = "Sequence Player (additive) 'IcicleLarva_Flight_Hovering_Idle_Up_And_Down_Xtra'"() {sequence=&IcicleLarva_Flight_Hovering_Idle_Up_And_Down_Xtra}
   n2 = Use cached pose 'Hover Input Pose'()
   Output Animation Pose(Result=Apply Additive(
     Base=Apply Additive(Base=@n2, Additive=@n1) {alphaBoolBlend {blendOption=Linear}}
     Additive=Blendspace Player 'IcicleLarva_Hover_BS'(X=0, Y=0) {blendSpace=&IcicleLarva_Hover_BS}
    ) {alphaBoolBlend {blendOption=Linear}})
  ]
 }
 Transition {
  path=Hover Adjustments.AnimGraphNode_StateMachine_0.HoverAdjustments SM.AnimStateTransitionNode_1.Transition
  nodes [
   Result(bCanEnterTransition=Get bIsAscend() -> (bIsAscend:bool))
   Get bIsPreAscend() -> (bIsPreAscend:bool)
  ]
 }
 Transition {
  path=Hover Adjustments.AnimGraphNode_StateMachine_0.HoverAdjustments SM.AnimStateTransitionNode_2.Transition
  nodes [Result(bCanEnterTransition=NOT Boolean(A=Get bIsPreAscend()))]
 }
 Hover Pivot {
  entry=Empty
  path=Hover Adjustments.AnimGraphNode_StateMachine_1.Hover Pivot
  ids=id
  nodes [
   AnimStateNode_0 = Empty() {
    is=title AnimStateNode
    stateEntered {guid=#2}
    stateFullyBlended {guid=#3}
    stateLeft {guid=#4}
   }
   AnimStateNode_2 = Pivot() {
    is=title AnimStateNode
    bAlwaysResetOnEntry=true
    stateEntered {guid=#5}
    stateFullyBlended {guid=#6}
    stateLeft {guid=#7}
   }
   AnimStateTransitionNode_0 = AnimStateTransitionNode() {crossfadeDuration=1, logicType=TLT_Inertialization}
   AnimStateTransitionNode_1 = AnimStateTransitionNode() {
    crossfadeDuration=1
    transitionEnd {guid=#8}
    transitionInterrupt {guid=#9}
    transitionStart {guid=#10}
   }
   AnimStateTransitionNode_7 = AnimStateTransitionNode() {
    crossfadeDuration=1
    transitionEnd {guid=#11}
    transitionInterrupt {guid=#12}
    transitionStart {guid=#13}
   }
  ]
  links [
   {
    dst=Pivot
    rule=Hover Adjustments.AnimGraphNode_StateMachine_1.Hover Pivot.AnimStateTransitionNode_0.Transition
    src=Empty
    via=AnimStateTransitionNode_0
   }
   {
    dst=Empty
    rule=Hover Adjustments.AnimGraphNode_StateMachine_1.Hover Pivot.AnimStateTransitionNode_1.Transition
    src=Pivot
    via=AnimStateTransitionNode_1
   }
   {
    dst=Empty
    rule=Hover Adjustments.AnimGraphNode_StateMachine_1.Hover Pivot.AnimStateTransitionNode_7.Transition
    src=Pivot
    via=AnimStateTransitionNode_7
   }
  ]
 }
 Empty {
  path=Hover Adjustments.AnimGraphNode_StateMachine_1.Hover Pivot.AnimStateNode_0.Empty
  nodes [Output Animation Pose(Result=Use cached pose 'Delta Pose'())]
 }
 Pivot {
  path=Hover Adjustments.AnimGraphNode_StateMachine_1.Hover Pivot.AnimStateNode_2.Pivot
  nodes [
   n1 = Sequence Player 'IcicleLarva_FlightTrans_Hover_Forward_to_Backwards'() {
    becomeRelevantFunction {bSelfContext=true, memberGuid=#14, memberName=OnBecomeRelevant_HoverPivot}
    node {
     bLoopAnimation=false
     bOverridePositionWhenJoiningSyncGroupAsLeader=true
     method=Graph
     sequence=/IcicleLarvaGameplay/Assets/Animation/Flight/Hover/Transitions/IcicleLarva_FlightTrans_Hover_Forward_to_Backwards
     becomeRelevantFunction {functionName=OnBecomeRelevant_HoverPivot}
    }
   }
   Output Animation Pose(Result=@n1)
  ]
 }
 Transition {
  path=Hover Adjustments.AnimGraphNode_StateMachine_1.Hover Pivot.AnimStateTransitionNode_0.Transition
  nodes [Result(bCanEnterTransition=Get bCanPivot() -> (bCanPivot:bool))]
 }
 Transition {
  path=Hover Adjustments.AnimGraphNode_StateMachine_1.Hover Pivot.AnimStateTransitionNode_1.Transition
  nodes [Result(bCanEnterTransition=Get bPivotToEmptyTransition() -> (bPivotToEmptyTransition:bool))]
 }
 Transition {
  path=Hover Adjustments.AnimGraphNode_StateMachine_1.Hover Pivot.AnimStateTransitionNode_7.Transition
  nodes [
   Result(bCanEnterTransition=Get bHoverPivotToEmptyTransition() -> (bHoverPivotToEmptyTransition:bool))
   "float < float"(A="Get Relevant Anim Time Remaining (Pivot)"!() -> (ReturnValue:float) {is=title K2Node_AnimGetter}, B=0.2)
  ]
 }
 ItemAimOffset {nodes [
   Input Pose!() -> (Pose, AimOffsetAlpha:double, Yaw:double, Pitch:double) {name=InPose_AimOffset}
   Input Pose!() -> (Pose) {name=InPose_PreAimOffsets}
   Output Pose() {layerGroup=Item Override}
  ]}
 ItemFallAdditive {nodes [
   Input Pose() {name=InPoseItemFallAdditive}
   Output Pose() {layerGroup=Item Override}
  ]}
 ItemFallLandAdditive {nodes [
   Input Pose() {name=InPoseItemFallLandAdditive}
   Output Pose() {layerGroup=Item Override}
  ]}
 ItemFallLevel2Override {nodes [
   Input Pose() {name=InPose_FallLevel2}
   Output Pose() {layerGroup=Item Override}
  ]}
 ItemFinalPoseOverride {nodes [
   n1 = "Transform (Modify) Bone - Bone: Root"!(
    ComponentPose=Local To Component(LocalPose=Use cached pose 'PostPrimarySM Pose'())
    Alpha=1
    Translation="0,0,0"
    Rotation_Roll:float=0
    Rotation_Pitch:float=0
    Rotation_Yaw:float=Get!() -> (RootInterpRotation_Roll:float, RootInterpRotation_Pitch:float, RootInterpRotation_Yaw:float).RootInterpRotation_Yaw
    Scale="1,1,1"
   ) -> (Pose) {rotationMode=BMM_Replace, rotationSpace=BCS_WorldSpace, alphaBoolBlend {blendOption=Linear}, boneToModify {boneName=Root}}
   n2 = Blend Poses by bool(BlendPose_0=Use cached pose 'PostPrimarySM Pose'(), BlendPose_1=Component To Local(ComponentPose=@n1), bActiveValue=Get bIsHover()) {blendType=Linear}
   Output Pose(Result=Blend Poses by bool(
     BlendPose_0=Modify Curve!(SourcePose=@n2, CurveValues_0=1, Alpha:float=1) -> (Pose) {curveNames [BothHandFK], curveValues [0]}
     BlendPose_1=Use cached pose 'PostPrimarySM Pose'()
     BlendTime_0=0.25
     BlendTime_1=0.25
     bActiveValue=Get bIsFlight() -> (bIsFlight:bool)
    ) {blendType=Linear}) {layerGroup=Item Override}
   AnimGraphNode_StateMachine_2 = Primary SM() {is=title AnimGraphNode_StateMachine}
   Save cached pose 'PostPrimarySM Pose'(Pose=All Flight Adjustments!(InPose_1:Pose Link Structure=@AnimGraphNode_StateMachine_2) -> (Pose:Pose Link Structure) {is=title AnimGraphNode_LinkedAnimLayer, layer=All Flight Adjustments}) {cacheName=PostPrimarySM Pose, cachePoseName=PreHeatVisionPose}
   AnimGraphNode_StateMachine_4 = Hover SM() {is=title AnimGraphNode_StateMachine, bAllowConduitEntryStates=true}
   Save cached pose 'Hover Pose'(Pose=Hover Adjustments!(InPose:Pose Link Structure=@AnimGraphNode_StateMachine_4) -> (Pose:Pose Link Structure) {is=title AnimGraphNode_LinkedAnimLayer, layer=Hover Adjustments}) {cacheName=Hover Pose}
   Save cached pose 'FinalPose'(Pose=Input Pose() {name=InFinalPose}) {cacheName=FinalPose, cachePoseName=FinalPose}
   AnimGraphNode_StateMachine_5 = NormalFlight SM() {is=title AnimGraphNode_StateMachine, bAllowConduitEntryStates=true}
   Save cached pose 'Flight Pose'(Pose=Normal Flight Adjustments!(FlightInPose:Pose Link Structure=@AnimGraphNode_StateMachine_5) -> (Pose:Pose Link Structure) {is=title AnimGraphNode_LinkedAnimLayer, layer=Normal Flight Adjustments}) {cacheName=Flight Pose}
   AnimGraphNode_StateMachine_6 = SprintFlight SM() {is=title AnimGraphNode_StateMachine, bAllowConduitEntryStates=true}
   Save cached pose 'SprintFlight Pose'(Pose=Sprint Flight Adjustments!(SprintInPose:Pose Link Structure=@AnimGraphNode_StateMachine_6) -> (Pose:Pose Link Structure) {is=title AnimGraphNode_LinkedAnimLayer, layer=Sprint Flight Adjustments}) {cacheName=SprintFlight Pose}
   AnimGraphNode_StateMachine_0 = ChargeFlight SM() {is=title AnimGraphNode_StateMachine, bAllowConduitEntryStates=true}
   Save cached pose 'Charge Pose'(Pose=Charge Flight Adjustments!(ChargeInPose:Pose Link Structure=@AnimGraphNode_StateMachine_0) -> (Pose:Pose Link Structure) {is=title AnimGraphNode_LinkedAnimLayer, layer=Charge Flight Adjustments}) {cacheName=Charge Pose}
   Save cached pose 'FlightBlend Pose'(Pose=Normal Flight Blend Adjustments!(InPose_0:Pose Link Structure=Sequence Player 'IcicleLarva_Flight_HandsBack_Idle_Loop'() {sequence=/IcicleLarvaGameplay/Assets/Animation/Flight/HandsBackFlight/IcicleLarva_Flight_HandsBack_Idle_Loop}) -> (Pose:Pose Link Structure) {is=title AnimGraphNode_LinkedAnimLayer, layer=Normal Flight Blend Adjustments}) {cacheName=FlightBlend Pose}
   Slot 'FullBody'() {slotName=FullBody}
  ]}
 ChargeFlight SM {
  entry=Conduit
  path=ItemFinalPoseOverride.AnimGraphNode_StateMachine_0.ChargeFlight SM
  ids=id
  nodes [
   To GroundSlam Prediction() {is=title AnimStateAliasNode, bGlobalAlias=true}
   AnimStateConduitNode_1 = Conduit()
   AnimStateNode_0 = GroundSlam Prediction() {is=title AnimStateNode}
   AnimStateNode_1 = Charge Flight() {is=title AnimStateNode}
   AnimStateNode_2 = From Hover Transition() {is=title AnimStateNode}
   AnimStateNode_3 = From Flight Transition() {is=title AnimStateNode}
   AnimStateNode_7 = Safety Transition() {is=title AnimStateNode}
   AnimStateNode_8 = From Sprint Transition() {is=title AnimStateNode}
   AnimStateTransitionNode_0 = AnimStateTransitionNode()
   AnimStateTransitionNode_1 = AnimStateTransitionNode() {bAutomaticRuleBasedOnSequencePlayerInState=true}
   AnimStateTransitionNode_10 = AnimStateTransitionNode() {bAutomaticRuleBasedOnSequencePlayerInState=true}
   AnimStateTransitionNode_2 = AnimStateTransitionNode() {priorityOrder=10}
   AnimStateTransitionNode_26 = AnimStateTransitionNode()
   AnimStateTransitionNode_29 = AnimStateTransitionNode()
   AnimStateTransitionNode_30 = AnimStateTransitionNode() {crossfadeDuration=0.01}
   AnimStateTransitionNode_4 = AnimStateTransitionNode() {bAutomaticRuleBasedOnSequencePlayerInState=true}
   AnimStateTransitionNode_5 = AnimStateTransitionNode() {bAutomaticRuleBasedOnSequencePlayerInState=true}
   AnimStateTransitionNode_8 = AnimStateTransitionNode() {bAutomaticRuleBasedOnSequencePlayerInState=true}
  ]
  links [
   {
    dst=From Flight Transition
    rule=ItemFinalPoseOverride.AnimGraphNode_StateMachine_0.ChargeFlight SM.AnimStateTransitionNode_29.AnimationTransitionGraph_0
    src=Conduit
    via=AnimStateTransitionNode_29
   }
   {
    dst=From Hover Transition
    rule=ItemFinalPoseOverride.AnimGraphNode_StateMachine_0.ChargeFlight SM.AnimStateTransitionNode_26.AnimationTransitionGraph_0
    src=Conduit
    via=AnimStateTransitionNode_26
   }
   {
    dst=From Sprint Transition
    rule=ItemFinalPoseOverride.AnimGraphNode_StateMachine_0.ChargeFlight SM.AnimStateTransitionNode_30.AnimationTransitionGraph_0
    src=Conduit
    via=AnimStateTransitionNode_30
   }
   {
    dst=Safety Transition
    rule=ItemFinalPoseOverride.AnimGraphNode_StateMachine_0.ChargeFlight SM.AnimStateTransitionNode_2.Transition
    src=Conduit
    via=AnimStateTransitionNode_2
   }
   {
    dst=Charge Flight
    rule=ItemFinalPoseOverride.AnimGraphNode_StateMachine_0.ChargeFlight SM.AnimStateTransitionNode_10.Transition
    src=From Flight Transition
    via=AnimStateTransitionNode_10
   }
   {
    dst=Charge Flight
    rule=ItemFinalPoseOverride.AnimGraphNode_StateMachine_0.ChargeFlight SM.AnimStateTransitionNode_8.Transition
    src=From Hover Transition
    via=AnimStateTransitionNode_8
   }
   {
    dst=Charge Flight
    rule=ItemFinalPoseOverride.AnimGraphNode_StateMachine_0.ChargeFlight SM.AnimStateTransitionNode_4.AnimationTransitionGraph_1
    src=From Sprint Transition
    via=AnimStateTransitionNode_4
   }
   {
    dst=Charge Flight
    rule=ItemFinalPoseOverride.AnimGraphNode_StateMachine_0.ChargeFlight SM.AnimStateTransitionNode_1.Transition
    src=GroundSlam Prediction
    via=AnimStateTransitionNode_1
   }
   {
    dst=Charge Flight
    rule=ItemFinalPoseOverride.AnimGraphNode_StateMachine_0.ChargeFlight SM.AnimStateTransitionNode_5.Transition
    src=Safety Transition
    via=AnimStateTransitionNode_5
   }
   {
    dst=GroundSlam Prediction
    rule=ItemFinalPoseOverride.AnimGraphNode_StateMachine_0.ChargeFlight SM.AnimStateTransitionNode_0.Transition
    src=To GroundSlam Prediction
    via=AnimStateTransitionNode_0
   }
  ]
 }
 Conduit {
  path=ItemFinalPoseOverride.AnimGraphNode_StateMachine_0.ChargeFlight SM.AnimStateConduitNode_1.Conduit
  nodes [Result(bCanEnterTransition=true)]
 }
 GroundSlam Prediction {
  path=ItemFinalPoseOverride.AnimGraphNode_StateMachine_0.ChargeFlight SM.AnimStateNode_0.GroundSlam Prediction
  nodes [
   Sequence Player 'IcicleLarva_Punch_InAir_RHand'() {bLoopAnimation=false, sequence=/IcicleLarvaGameplay/Assets/Animation/Proto/IcicleLarva_Punch_InAir_RHand}
   Output Animation Pose(Result=Sequence Player 'IcicleLarva_GroundSlam_Prediction'() {bLoopAnimation=false, playRate=0.5, sequence=/IcicleLarvaGameplay/Assets/Animation/Proto/IcicleLarva_GroundSlam_Prediction})
  ]
 }
 Charge Flight {
  path=ItemFinalPoseOverride.AnimGraphNode_StateMachine_0.ChargeFlight SM.AnimStateNode_1.Charge Flight
  nodes [
   Slot 'FullBody'() {slotName=FullBody}
   Output Animation Pose(Result=Sequence Player 'IcicleLarva_AirChargeLoop_Proto'() {sequence=/IcicleLarvaGameplay/Assets/Animation/Proto/IcicleLarva_AirChargeLoop_Proto})
  ]
 }
 From Hover Transition {
  path=ItemFinalPoseOverride.AnimGraphNode_StateMachine_0.ChargeFlight SM.AnimStateNode_2.From Hover Transition
  nodes [Output Animation Pose(Result=Modify Curve(
     SourcePose=Sequence Player 'IcicleLarva_Flight_Transition_HoveringToChargeFly'() {bLoopAnimation=false, sequence=/IcicleLarvaGameplay/Assets/Animation/Proto/IcicleLarva_Flight_Transition_HoveringToChargeFly}
     CurveValues_0=0
     Alpha=1
    ) {curveNames [HoverPivotAlpha], curveValues [0]}) {
    becomeRelevantFunction {bSelfContext=true, memberGuid=#15, memberName=FromHoverFalse}
    node {becomeRelevantFunction {functionName=FromHoverFalse}}
   }]
 }
 From Flight Transition {
  path=ItemFinalPoseOverride.AnimGraphNode_StateMachine_0.ChargeFlight SM.AnimStateNode_3.From Flight Transition
  nodes [Output Animation Pose(Result=Sequence Player 'IcicleLarva_Flight_Transition_Flight_to_AirCharge'() {bLoopAnimation=false, sequence=/IcicleLarvaGameplay/Assets/Animation/Flight/IcicleLarva_Flight_Transition_Flight_to_AirCharge}) {
    becomeRelevantFunction {bSelfContext=true, memberGuid=#16, memberName=FromFlightFalse}
    node {becomeRelevantFunction {functionName=FromFlightFalse}}
   }]
 }
 Safety Transition {
  path=ItemFinalPoseOverride.AnimGraphNode_StateMachine_0.ChargeFlight SM.AnimStateNode_7.Safety Transition
  nodes [Output Animation Pose(Result=Sequence Player 'IcicleLarva_Flight_Transition_HoveringToChargeFly'() {bLoopAnimation=false, sequence=&IcicleLarva_Flight_Transition_HoveringToChargeFly}) {
    becomeRelevantFunction {bSelfContext=true, memberGuid=#15, memberName=FromHoverFalse}
    node {becomeRelevantFunction {functionName=FromHoverFalse}}
   }]
 }
 From Sprint Transition {
  path=ItemFinalPoseOverride.AnimGraphNode_StateMachine_0.ChargeFlight SM.AnimStateNode_8.From Sprint Transition
  nodes [
   n1 = Sequence Player 'IcicleLarva_Flight_Transition_OneHandedToChargeFly'() {
    bLoopAnimation=false
    sequence=/IcicleLarvaGameplay/Assets/Animation/Flight/OneHandedFlight/IcicleLarva_Flight_Transition_OneHandedToChargeFly
   }
   Output Animation Pose(Result=@n1) {
    becomeRelevantFunction {bSelfContext=true, memberGuid=#17, memberName=FromSprintFalse}
    node {becomeRelevantFunction {functionName=FromSprintFalse}}
   }
  ]
 }
 Transition {
  path=ItemFinalPoseOverride.AnimGraphNode_StateMachine_0.ChargeFlight SM.AnimStateTransitionNode_0.Transition
  nodes [Result(bCanEnterTransition=Get bIsGroundSlamPredicted() -> (bIsGroundSlamPredicted:bool))]
 }
 Transition {path=ItemFinalPoseOverride.AnimGraphNode_StateMachine_0.ChargeFlight SM.AnimStateTransitionNode_1.Transition, nodes [Result()]}
 Transition {path=ItemFinalPoseOverride.AnimGraphNode_StateMachine_0.ChargeFlight SM.AnimStateTransitionNode_10.Transition, nodes [Result()]}
 Transition {
  path=ItemFinalPoseOverride.AnimGraphNode_StateMachine_0.ChargeFlight SM.AnimStateTransitionNode_2.Transition
  nodes [Result(bCanEnterTransition=true)]
 }
 AnimationTransitionGraph_0 {
  path=ItemFinalPoseOverride.AnimGraphNode_StateMachine_0.ChargeFlight SM.AnimStateTransitionNode_26.AnimationTransitionGraph_0
  nodes [Result(bCanEnterTransition=Get bIsFromHover() -> (bIsFromHover:bool))]
 }
 AnimationTransitionGraph_0 {
  path=ItemFinalPoseOverride.AnimGraphNode_StateMachine_0.ChargeFlight SM.AnimStateTransitionNode_29.AnimationTransitionGraph_0
  nodes [Result(bCanEnterTransition=Get bIsFromFlight() -> (bIsFromFlight:bool))]
 }
 AnimationTransitionGraph_0 {
  path=ItemFinalPoseOverride.AnimGraphNode_StateMachine_0.ChargeFlight SM.AnimStateTransitionNode_30.AnimationTransitionGraph_0
  nodes [Result(bCanEnterTransition=Get bIsFromSprint() -> (bIsFromSprint:bool))]
 }
 AnimationTransitionGraph_1 {
  path=ItemFinalPoseOverride.AnimGraphNode_StateMachine_0.ChargeFlight SM.AnimStateTransitionNode_4.AnimationTransitionGraph_1
  nodes [Result()]
 }
 Transition {path=ItemFinalPoseOverride.AnimGraphNode_StateMachine_0.ChargeFlight SM.AnimStateTransitionNode_5.Transition, nodes [Result()]}
 Transition {path=ItemFinalPoseOverride.AnimGraphNode_StateMachine_0.ChargeFlight SM.AnimStateTransitionNode_8.Transition, nodes [Result()]}
 Primary SM {
  entry=Passthrough
  path=ItemFinalPoseOverride.AnimGraphNode_StateMachine_2.Primary SM
  ids=id
  nodes [
   To Hover Landing() {is=title AnimStateAliasNode, also {aliases [Hover]}}
   To Flight Pre-Landing() {is=title AnimStateAliasNode, also {aliases [Flight,Sprint]}}
   IsWalking() {is=title AnimStateAliasNode, also {aliases [Charge,Hover,Sprint]}}
   From Hover() {is=title AnimStateAliasNode, also {aliases [Hover, Punch Outro]}}
   From Charge() {is=title AnimStateAliasNode, also {aliases [Charge]}}
   Is Outro() {is=title AnimStateAliasNode, also {aliases [Charge, Flight, Flight Pre-Landing, Hover, Hover Landing, Sprint]}}
   AnimStateConduitNode_0 = Conduit()
   AnimStateNode_0 = GroundSlam Passthrough() {
    is=title AnimStateNode
    stateEntered {guid=#18}
    stateFullyBlended {guid=#19}
    stateLeft {guid=#20}
   }
   AnimStateNode_1 = Punch Outro() {
    is=title AnimStateNode
    stateEntered {guid=#18}
    stateFullyBlended {guid=#19}
    stateLeft {guid=#20}
   }
   AnimStateNode_10 = Punch Intro() {
    is=title AnimStateNode
    stateEntered {guid=#18}
    stateFullyBlended {guid=#19}
    stateLeft {guid=#20}
   }
   AnimStateNode_11 = PreAscend() {is=title AnimStateNode}
   AnimStateNode_2 = Flight Pre-Landing() {is=title AnimStateNode}
   AnimStateNode_3 = Sprint() {is=title AnimStateNode}
   AnimStateNode_4 = Charge() {is=title AnimStateNode}
   AnimStateNode_5 = Hover Landing() {is=title AnimStateNode}
   AnimStateNode_6 = Outro() {
    is=title AnimStateNode
    stateEntered {guid=#18}
    stateFullyBlended {guid=#19}
    stateLeft {guid=#20}
   }
   AnimStateNode_7 = Flight() {is=title AnimStateNode}
   AnimStateNode_8 = Hover()
   AnimStateNode_9 = Passthrough() {
    stateEntered {guid=#18}
    stateFullyBlended {guid=#19}
    stateLeft {guid=#20}
   }
   AnimStateTransitionNode_0 = AnimStateTransitionNode() {crossfadeDuration=0}
   AnimStateTransitionNode_1 = AnimStateTransitionNode()
   AnimStateTransitionNode_11 = AnimStateTransitionNode() {bAutomaticRuleBasedOnSequencePlayerInState=true, crossfadeDuration=0.5}
   AnimStateTransitionNode_12 = AnimStateTransitionNode() {crossfadeDuration=0.25}
   AnimStateTransitionNode_13 = AnimStateTransitionNode()
   AnimStateTransitionNode_14 = AnimStateTransitionNode()
   AnimStateTransitionNode_15 = AnimStateTransitionNode() {crossfadeDuration=0.4, priorityOrder=10}
   AnimStateTransitionNode_16 = AnimStateTransitionNode()
   AnimStateTransitionNode_17 = AnimStateTransitionNode() {crossfadeDuration=0.08}
   AnimStateTransitionNode_18 = AnimStateTransitionNode()
   AnimStateTransitionNode_19 = AnimStateTransitionNode() {crossfadeDuration=0.05}
   AnimStateTransitionNode_2 = AnimStateTransitionNode() {priorityOrder=10}
   AnimStateTransitionNode_21 = AnimStateTransitionNode()
   AnimStateTransitionNode_22 = AnimStateTransitionNode()
   AnimStateTransitionNode_23 = AnimStateTransitionNode() {crossfadeDuration=0.6}
   AnimStateTransitionNode_24 = AnimStateTransitionNode()
   AnimStateTransitionNode_25 = AnimStateTransitionNode()
   AnimStateTransitionNode_28 = AnimStateTransitionNode() {bAutomaticRuleBasedOnSequencePlayerInState=true, crossfadeDuration=0.4}
   AnimStateTransitionNode_29 = AnimStateTransitionNode() {crossfadeDuration=0.4}
   AnimStateTransitionNode_3 = AnimStateTransitionNode() {crossfadeDuration=0.1}
   AnimStateTransitionNode_30 = AnimStateTransitionNode() {bAutomaticRuleBasedOnSequencePlayerInState=true, crossfadeDuration=0.6}
   AnimStateTransitionNode_31 = AnimStateTransitionNode()
   AnimStateTransitionNode_32 = AnimStateTransitionNode()
   AnimStateTransitionNode_4 = AnimStateTransitionNode()
   AnimStateTransitionNode_5 = AnimStateTransitionNode() {crossfadeDuration=0.25}
   AnimStateTransitionNode_6 = AnimStateTransitionNode()
   AnimStateTransitionNode_7 = AnimStateTransitionNode() {crossfadeDuration=1}
   AnimStateTransitionNode_8 = AnimStateTransitionNode() {bAutomaticRuleBasedOnSequencePlayerInState=true, crossfadeDuration=0.3}
   AnimStateTransitionNode_9 = AnimStateTransitionNode()
  ]
  links [
   {
    dst=Flight
    rule=ItemFinalPoseOverride.AnimGraphNode_StateMachine_2.Primary SM.AnimStateTransitionNode_24.Transition
    src=Charge
    via=AnimStateTransitionNode_24
   }
   {
    dst=GroundSlam Passthrough
    rule=ItemFinalPoseOverride.AnimGraphNode_StateMachine_2.Primary SM.AnimStateTransitionNode_0.Transition
    src=Charge
    via=AnimStateTransitionNode_0
   }
   {
    dst=Sprint
    rule=ItemFinalPoseOverride.AnimGraphNode_StateMachine_2.Primary SM.AnimStateTransitionNode_25.Transition
    src=Charge
    via=AnimStateTransitionNode_25
   }
   {
    dst=Hover
    rule=ItemFinalPoseOverride.AnimGraphNode_StateMachine_2.Primary SM.AnimStateTransitionNode_29.Transition
    src=Conduit
    via=AnimStateTransitionNode_29
   }
   {
    dst=Passthrough
    rule=ItemFinalPoseOverride.AnimGraphNode_StateMachine_2.Primary SM.AnimStateTransitionNode_30.Transition
    src=Conduit
    via=AnimStateTransitionNode_30
   }
   {
    dst=Charge
    rule=ItemFinalPoseOverride.AnimGraphNode_StateMachine_2.Primary SM.AnimStateTransitionNode_13.Transition
    src=Flight
    via=AnimStateTransitionNode_13
   }
   {
    dst=Hover
    rule=ItemFinalPoseOverride.AnimGraphNode_StateMachine_2.Primary SM.AnimStateTransitionNode_23.Transition
    src=Flight
    via=AnimStateTransitionNode_23
   }
   {
    dst=Sprint
    rule=ItemFinalPoseOverride.AnimGraphNode_StateMachine_2.Primary SM.AnimStateTransitionNode_17.Transition
    src=Flight
    via=AnimStateTransitionNode_17
   }
   {
    dst=Hover Landing
    rule=ItemFinalPoseOverride.AnimGraphNode_StateMachine_2.Primary SM.AnimStateTransitionNode_8.AnimationTransitionGraph_0
    src=Flight Pre-Landing
    via=AnimStateTransitionNode_8
   }
   {
    dst=Hover
    rule=ItemFinalPoseOverride.AnimGraphNode_StateMachine_2.Primary SM.AnimStateTransitionNode_19.Transition
    src=From Charge
    via=AnimStateTransitionNode_19
   }
   {
    dst=Charge
    rule=ItemFinalPoseOverride.AnimGraphNode_StateMachine_2.Primary SM.AnimStateTransitionNode_18.Transition
    src=From Hover
    via=AnimStateTransitionNode_18
   }
   {
    dst=Passthrough
    rule=ItemFinalPoseOverride.AnimGraphNode_StateMachine_2.Primary SM.AnimStateTransitionNode_1.Transition
    src=GroundSlam Passthrough
    via=AnimStateTransitionNode_1
   }
   {
    dst=Flight
    rule=ItemFinalPoseOverride.AnimGraphNode_StateMachine_2.Primary SM.AnimStateTransitionNode_9.Transition
    src=Hover
    via=AnimStateTransitionNode_9
   }
   {
    dst=Sprint
    rule=ItemFinalPoseOverride.AnimGraphNode_StateMachine_2.Primary SM.AnimStateTransitionNode_4.Transition
    src=Hover
    via=AnimStateTransitionNode_4
   }
   {
    dst=Passthrough
    rule=ItemFinalPoseOverride.AnimGraphNode_StateMachine_2.Primary SM.AnimStateTransitionNode_11.Transition
    src=Hover Landing
    via=AnimStateTransitionNode_11
   }
   {
    dst=Outro
    rule=ItemFinalPoseOverride.AnimGraphNode_StateMachine_2.Primary SM.AnimStateTransitionNode_12.Transition
    src=Is Outro
    via=AnimStateTransitionNode_12
   }
   {
    dst=Passthrough
    rule=ItemFinalPoseOverride.AnimGraphNode_StateMachine_2.Primary SM.AnimStateTransitionNode_2.Transition
    src=IsWalking
    via=AnimStateTransitionNode_2
   }
   {
    dst=Hover
    rule=ItemFinalPoseOverride.AnimGraphNode_StateMachine_2.Primary SM.AnimStateTransitionNode_15.Transition
    src=Passthrough
    via=AnimStateTransitionNode_15
   }
   {
    dst=PreAscend
    rule=ItemFinalPoseOverride.AnimGraphNode_StateMachine_2.Primary SM.AnimStateTransitionNode_21.Transition
    src=Passthrough
    via=AnimStateTransitionNode_21
   }
   {
    dst=Punch Intro
    rule=ItemFinalPoseOverride.AnimGraphNode_StateMachine_2.Primary SM.AnimStateTransitionNode_3.Transition
    src=Passthrough
    via=AnimStateTransitionNode_3
   }
   {
    dst=Hover
    rule=ItemFinalPoseOverride.AnimGraphNode_StateMachine_2.Primary SM.AnimStateTransitionNode_28.Transition
    src=PreAscend
    via=AnimStateTransitionNode_28
   }
   {
    dst=Passthrough
    rule=ItemFinalPoseOverride.AnimGraphNode_StateMachine_2.Primary SM.AnimStateTransitionNode_16.Transition
    src=Punch Intro
    via=AnimStateTransitionNode_16
   }
   {
    dst=Punch Outro
    rule=ItemFinalPoseOverride.AnimGraphNode_StateMachine_2.Primary SM.AnimStateTransitionNode_6.Transition
    src=Punch Intro
    via=AnimStateTransitionNode_6
   }
   {
    dst=Conduit
    rule=ItemFinalPoseOverride.AnimGraphNode_StateMachine_2.Primary SM.AnimStateTransitionNode_31.Transition
    src=Punch Outro
    via=AnimStateTransitionNode_31
   }
   {
    dst=Charge
    rule=ItemFinalPoseOverride.AnimGraphNode_StateMachine_2.Primary SM.AnimStateTransitionNode_14.Transition
    src=Sprint
    via=AnimStateTransitionNode_14
   }
   {
    dst=Flight
    rule=ItemFinalPoseOverride.AnimGraphNode_StateMachine_2.Primary SM.AnimStateTransitionNode_32.Transition
    src=Sprint
    via=AnimStateTransitionNode_32
   }
   {
    dst=Hover
    rule=ItemFinalPoseOverride.AnimGraphNode_StateMachine_2.Primary SM.AnimStateTransitionNode_22.Transition
    src=Sprint
    via=AnimStateTransitionNode_22
   }
   {
    dst=Flight Pre-Landing
    rule=ItemFinalPoseOverride.AnimGraphNode_StateMachine_2.Primary SM.AnimStateTransitionNode_7.AnimationTransitionGraph_0
    src=To Flight Pre-Landing
    via=AnimStateTransitionNode_7
   }
   {
    dst=Hover Landing
    rule=ItemFinalPoseOverride.AnimGraphNode_StateMachine_2.Primary SM.AnimStateTransitionNode_5.AnimationTransitionGraph_0
    src=To Hover Landing
    via=AnimStateTransitionNode_5
   }
  ]
 }
 Conduit {path=ItemFinalPoseOverride.AnimGraphNode_StateMachine_2.Primary SM.AnimStateConduitNode_0.Conduit, nodes [Result(bCanEnterTransition=true)]}
 GroundSlam Passthrough {
  path=ItemFinalPoseOverride.AnimGraphNode_StateMachine_2.Primary SM.AnimStateNode_0.GroundSlam Passthrough
  nodes [Output Animation Pose(Result=Modify Curve!(SourcePose=Use cached pose 'FinalPose'(), CurveValues_0=1, CurveValues_1:float=1, Alpha=1) -> (Pose) {curveNames [DisableFlightPitchAdditive, DisableFlightYawAdditive], curveValues [0,0]})]
 }
 Punch Outro {
  path=ItemFinalPoseOverride.AnimGraphNode_StateMachine_2.Primary SM.AnimStateNode_1.Punch Outro
  nodes [
   n1 = Sequence Player 'IcicleLarva_Punch_GroundToAir_Outro'() {
    bLoopAnimation=false
    groupName=Punch
    method=SyncGroup
    sequence=/IcicleLarvaGameplay/Assets/Animation/Proto/IcicleLarva_Punch_GroundToAir_Outro
   }
   n2 = Sequence Player 'IcicleLarva_Punch_OnGround_Outro_RootMotion'() {
    bLoopAnimation=false
    groupName=Punch
    method=SyncGroup
    sequence=/IcicleLarvaGameplay/Assets/Animation/Proto/IcicleLarva_Punch_OnGround_Outro_RootMotion
   }
   n3 = Two Way Blend(A=Use cached pose 'FinalPose'(), B=@n2, Alpha=0.01) {blendNode {bAlwaysUpdateChildren=true, alphaBoolBlend {blendOption=Linear}}}
   n4 = Get bIsPunchInAir() -> (bIsPunchInAir:bool)
   Output Animation Pose(Result=Modify Curve(
     SourcePose=Blend Poses by bool(BlendPose_0=@n1, BlendPose_1=@n3, BlendTime_0=0.2, BlendTime_1=0.2, bActiveValue=@n4) {blendType=Linear}
     CurveValues_0=1
     CurveValues_1=1
     Alpha=1
    ) {curveNames [DisableFlightPitchAdditive, DisableFlightYawAdditive], curveValues [0,0]})
  ]
 }
 Punch Intro {
  path=ItemFinalPoseOverride.AnimGraphNode_StateMachine_2.Primary SM.AnimStateNode_10.Punch Intro
  nodes [
   n1 = Sequence Player 'IcicleLarva_Punch_FromTacSprint_Intro'() {bLoopAnimation=false, playRate=0.65, sequence=/IcicleLarvaGameplay/Assets/Animation/Proto/IcicleLarva_Punch_FromTacSprint_Intro}
   n2 = Blend Poses by bool(
    BlendPose_0=@n1
    BlendPose_1=Use cached pose 'FinalPose'()
    BlendTime_0=0
    BlendTime_1=0.2
    bActiveValue=Get bIsPunchTacSprint() -> (bIsPunchTacSprint:bool)
   ) {blendType=Linear}
   n3 = Get!() -> (AdjustedAim_ResultingYaw:float=~, AdjustedAim_ResultingPitch:float=~, AdjustedAim_ZeroOutPitchWeightCurveName:Name)
   Output Animation Pose(Result=Modify Curve(
     SourcePose=AimOffset Player 'IcicleLarva_Punch_AO'!(X:float=@n3, Y:float=@n3.AdjustedAim_ResultingPitch, BasePose:Pose Link Structure=@n2, Alpha:float=1) -> (Pose:Pose Link Structure) {blendSpace=/IcicleLarvaGameplay/Assets/Animation/Proto/Punch_AO/IcicleLarva_Punch_AO, alphaBoolBlend {blendOption=Linear}}
     CurveValues_0=1
     CurveValues_1=1
     Alpha=1
    ) {curveNames [DisableFlightPitchAdditive, DisableFlightYawAdditive], curveValues [0,0]})
  ]
 }
 PreAscend {
  path=ItemFinalPoseOverride.AnimGraphNode_StateMachine_2.Primary SM.AnimStateNode_11.PreAscend
  nodes [
   n1 = "Sequence Player (additive) 'IcicleLarva_Ascend_Squat_Additive'"!(PlayRate:float=1.1) -> (Pose) {bLoopAnimation=false, sequence=/IcicleLarvaGameplay/Assets/Animation/Proto/IcicleLarva_Ascend_Squat_Additive}
   n2 = Use cached pose 'FinalPose'()
   Output Animation Pose(Result=Modify Curve(
     SourcePose=Apply Additive(Base=@n2, Additive=@n1, Alpha=0.01) {alphaBoolBlend {blendOption=Linear}}
     CurveValues_0=1
     CurveValues_1=1
     Alpha=1
    ) {curveNames [DisableFlightPitchAdditive, DisableFlightYawAdditive], curveValues [0,0]})
  ]
 }
 Flight Pre-Landing {
  path=ItemFinalPoseOverride.AnimGraphNode_StateMachine_2.Primary SM.AnimStateNode_2.Flight Pre-Landing
  nodes [
   n1 = Sequence Player 'IcicleLarva_FlightTrans_ForwardToCenterHover'() {bLoopAnimation=false, sequence=/IcicleLarvaGameplay/Assets/Animation/Flight/Hover/Transitions/IcicleLarva_FlightTrans_ForwardToCenterHover}
   Output Animation Pose(Result=@n1)
  ]
 }
 Sprint {
  path=ItemFinalPoseOverride.AnimGraphNode_StateMachine_2.Primary SM.AnimStateNode_3.Sprint
  nodes [Output Animation Pose(Result=Use cached pose 'SprintFlight Pose'()) {
    node {stateFullyBlendedOutFunction {functionName=FromSprintFalse}, updateFunction {functionName=FromSprintTrue}}
    stateFullyBlendedOutFunction {bSelfContext=true, memberGuid=#17, memberName=FromSprintFalse}
    updateFunction {bSelfContext=true, memberGuid=#21, memberName=FromSprintTrue}
   }]
 }
 Charge {
  path=ItemFinalPoseOverride.AnimGraphNode_StateMachine_2.Primary SM.AnimStateNode_4.Charge
  nodes [Output Animation Pose(Result=Use cached pose 'Charge Pose'()) {
    node {stateFullyBlendedOutFunction {functionName=FromChargeFalse}, updateFunction {functionName=FromChargeTrue}}
    stateFullyBlendedOutFunction {bSelfContext=true, memberGuid=#22, memberName=FromChargeFalse}
    updateFunction {bSelfContext=true, memberGuid=#23, memberName=FromChargeTrue}
   }]
 }
 Hover Landing {
  path=ItemFinalPoseOverride.AnimGraphNode_StateMachine_2.Primary SM.AnimStateNode_5.Hover Landing
  nodes [
   n1 = Sequence Player 'IcicleLarva_Flight_Transition_HoveringToStand'() {
    bLoopAnimation=false
    groupName=HoverLand
    method=SyncGroup
    sequence=/IcicleLarvaGameplay/Assets/Animation/Proto/IcicleLarva_Flight_Transition_HoveringToStand
   }
   n2 = Layered blend per bone(BasePose=Use cached pose 'FinalPose'(), BlendPoses_0=@n1) {bMeshSpaceRotationBlend=true, layerSetup [{branchFilters [{blendDepth=4, boneName=spine_01}]}]}
   n3 = Sequence Player 'IcicleLarva_Flight_Transition_HoveringToStand'() {
    bLoopAnimation=false
    groupName=HoverLand
    method=SyncGroup
    sequence=&IcicleLarva_Flight_Transition_HoveringToStand
   }
   n4 = Get bIsMoving2D()
   Output Animation Pose(Result=Modify Curve(
     SourcePose=Blend Poses by bool(BlendPose_0=@n2, BlendPose_1=@n3, BlendTime_0=0.2, BlendTime_1=0.2, bActiveValue=@n4) {blendType=Linear}
     CurveValues_0=1
     CurveValues_1=1
     Alpha=1
    ) {curveNames [DisableFlightPitchAdditive, DisableFlightYawAdditive], curveValues [0,0]})
  ]
 }
 Outro {
  path=ItemFinalPoseOverride.AnimGraphNode_StateMachine_2.Primary SM.AnimStateNode_6.Outro
  nodes [Output Animation Pose(Result=Use cached pose 'FinalPose'())]
 }
 Flight {
  path=ItemFinalPoseOverride.AnimGraphNode_StateMachine_2.Primary SM.AnimStateNode_7.Flight
  nodes [Output Animation Pose(Result=Use cached pose 'Flight Pose'()) {
    node {stateFullyBlendedOutFunction {functionName=FromFlightFalse}, updateFunction {functionName=FromFlightTrue}}
    stateFullyBlendedOutFunction {bSelfContext=true, memberGuid=#16, memberName=FromFlightFalse}
    updateFunction {bSelfContext=true, memberGuid=#24, memberName=FromFlightTrue}
   }]
 }
 Hover {
  path=ItemFinalPoseOverride.AnimGraphNode_StateMachine_2.Primary SM.AnimStateNode_8.Hover
  nodes [
   AnimGraphNode_BlendSpaceGraph_0 = Blendspace 'IcicleLarva_Flight_OneHanded_Speed_BS'!(X:float=0) -> (Pose:Pose Link Structure) {is=title AnimGraphNode_BlendSpaceGraph}
   Output Animation Pose(Result=Slot 'FullBody'(Source=@AnimGraphNode_BlendSpaceGraph_0) {slotName=FullBody}) {
    node {stateFullyBlendedOutFunction {functionName=FromHoverFalse}, updateFunction {functionName=FromHoverTrue}}
    stateFullyBlendedOutFunction {bSelfContext=true, memberGuid=#15, memberName=FromHoverFalse}
    updateFunction {bSelfContext=true, memberGuid=#25, memberName=FromHoverTrue}
   }
  ]
 }
 IcicleLarva_Flight_OneHanded_Speed_BS {
  path=ItemFinalPoseOverride.AnimGraphNode_StateMachine_2.Primary SM.AnimStateNode_8.Hover.AnimGraphNode_BlendSpaceGraph_0.IcicleLarva_Flight_OneHanded_Speed_BS
  links []
  nodes []
 }
 IcicleLarva_Flight_OneHanded_BasePose_Xtra_Additive {
  path=ItemFinalPoseOverride.AnimGraphNode_StateMachine_2.Primary SM.AnimStateNode_8.Hover.AnimGraphNode_BlendSpaceGraph_0.IcicleLarva_Flight_OneHanded_Speed_BS.IcicleLarva_Flight_OneHanded_BasePose_Xtra_Additive
  nodes [Output Animation Pose!(Result:Pose Link Structure=Use cached pose 'Hover Pose'()) {is=title AnimGraphNode_BlendSpaceSampleResult}]
 }
 IcicleLarva_Flight_OneHanded_Center_Old {
  path=ItemFinalPoseOverride.AnimGraphNode_StateMachine_2.Primary SM.AnimStateNode_8.Hover.AnimGraphNode_BlendSpaceGraph_0.IcicleLarva_Flight_OneHanded_Speed_BS.IcicleLarva_Flight_OneHanded_Center_Old
  nodes [Output Animation Pose(Result=Use cached pose 'FlightBlend Pose'())]
 }
 Passthrough {
  path=ItemFinalPoseOverride.AnimGraphNode_StateMachine_2.Primary SM.AnimStateNode_9.Passthrough
  nodes [Output Animation Pose(Result=Use cached pose 'FinalPose'()) {is=title AnimGraphNode_StateResult}]
 }
 Transition {
  path=ItemFinalPoseOverride.AnimGraphNode_StateMachine_2.Primary SM.AnimStateTransitionNode_0.Transition
  nodes [Result(bCanEnterTransition=Get bIsGroundSlamMontage())]
 }
 Transition {
  path=ItemFinalPoseOverride.AnimGraphNode_StateMachine_2.Primary SM.AnimStateTransitionNode_1.Transition
  nodes [Result(bCanEnterTransition=NOT Boolean(A=Get bIsGroundSlamMontage()))]
 }
 Transition {path=ItemFinalPoseOverride.AnimGraphNode_StateMachine_2.Primary SM.AnimStateTransitionNode_11.Transition, nodes [Result()]}
 Transition {
  path=ItemFinalPoseOverride.AnimGraphNode_StateMachine_2.Primary SM.AnimStateTransitionNode_12.Transition
  nodes [Result(bCanEnterTransition=Get bIsOutro() -> (bIsOutro:bool))]
 }
 Transition {
  path=ItemFinalPoseOverride.AnimGraphNode_StateMachine_2.Primary SM.AnimStateTransitionNode_13.Transition
  nodes [Result(bCanEnterTransition=Get bIsAirCharging())]
 }
 Transition {
  path=ItemFinalPoseOverride.AnimGraphNode_StateMachine_2.Primary SM.AnimStateTransitionNode_14.Transition
  nodes [Result(bCanEnterTransition=Get bIsAirCharging())]
 }
 Transition {
  path=ItemFinalPoseOverride.AnimGraphNode_StateMachine_2.Primary SM.AnimStateTransitionNode_15.Transition
  nodes [Result(bCanEnterTransition=Get bIsFlight())]
 }
 Transition {
  path=ItemFinalPoseOverride.AnimGraphNode_StateMachine_2.Primary SM.AnimStateTransitionNode_16.Transition
  nodes [Result(bCanEnterTransition=NOT Boolean(A=Get bIsPunch() -> (bIsPunch:bool)))]
 }
 Transition {
  path=ItemFinalPoseOverride.AnimGraphNode_StateMachine_2.Primary SM.AnimStateTransitionNode_17.Transition
  nodes [Result(bCanEnterTransition=Get bIsSprintFlight())]
 }
 Transition {
  path=ItemFinalPoseOverride.AnimGraphNode_StateMachine_2.Primary SM.AnimStateTransitionNode_18.Transition
  nodes [Result(bCanEnterTransition=Get bIsAirCharging())]
 }
 Transition {
  path=ItemFinalPoseOverride.AnimGraphNode_StateMachine_2.Primary SM.AnimStateTransitionNode_19.Transition
  nodes [Result(bCanEnterTransition=Get bIsHover())]
 }
 Transition {
  path=ItemFinalPoseOverride.AnimGraphNode_StateMachine_2.Primary SM.AnimStateTransitionNode_2.Transition
  nodes [
   Result(bCanEnterTransition=Get bIsGroundedSafety() -> (bIsGroundedSafety:bool))
   Get bIsWalkingIcicleLarva() -> (bIsWalkingIcicleLarva:bool)
  ]
 }
 Transition {
  path=ItemFinalPoseOverride.AnimGraphNode_StateMachine_2.Primary SM.AnimStateTransitionNode_21.Transition
  nodes [Result(bCanEnterTransition=Get bIsPreAscend())]
 }
 Transition {
  path=ItemFinalPoseOverride.AnimGraphNode_StateMachine_2.Primary SM.AnimStateTransitionNode_22.Transition
  nodes [Result(bCanEnterTransition=Get bIsHover())]
 }
 Transition {
  path=ItemFinalPoseOverride.AnimGraphNode_StateMachine_2.Primary SM.AnimStateTransitionNode_23.Transition
  nodes [Result(bCanEnterTransition=Get bIsHover())]
 }
 Transition {
  path=ItemFinalPoseOverride.AnimGraphNode_StateMachine_2.Primary SM.AnimStateTransitionNode_24.Transition
  nodes [Result(bCanEnterTransition=Get bIsNormalFlight())]
 }
 Transition {
  path=ItemFinalPoseOverride.AnimGraphNode_StateMachine_2.Primary SM.AnimStateTransitionNode_25.Transition
  nodes [Result(bCanEnterTransition=Get bIsSprintFlight())]
 }
 Transition {path=ItemFinalPoseOverride.AnimGraphNode_StateMachine_2.Primary SM.AnimStateTransitionNode_28.Transition, nodes [Result()]}
 Transition {
  path=ItemFinalPoseOverride.AnimGraphNode_StateMachine_2.Primary SM.AnimStateTransitionNode_29.Transition
  nodes [Result(bCanEnterTransition=Get bIsFlight())]
 }
 Transition {
  path=ItemFinalPoseOverride.AnimGraphNode_StateMachine_2.Primary SM.AnimStateTransitionNode_3.Transition
  nodes [Result(bCanEnterTransition=Get bIsPunchMomentary() -> (bIsPunchMomentary:bool))]
 }
 Transition {path=ItemFinalPoseOverride.AnimGraphNode_StateMachine_2.Primary SM.AnimStateTransitionNode_30.Transition, nodes [Result()]}
 Transition {
  path=ItemFinalPoseOverride.AnimGraphNode_StateMachine_2.Primary SM.AnimStateTransitionNode_31.Transition
  nodes [Result(bCanEnterTransition=NOT Boolean(A=Get bIsPunch()))]
 }
 Transition {
  path=ItemFinalPoseOverride.AnimGraphNode_StateMachine_2.Primary SM.AnimStateTransitionNode_32.Transition
  nodes [Result(bCanEnterTransition=Get bIsNormalFlight())]
 }
 Transition {
  path=ItemFinalPoseOverride.AnimGraphNode_StateMachine_2.Primary SM.AnimStateTransitionNode_4.Transition
  nodes [Result(bCanEnterTransition=Get bIsSprintFlight())]
 }
 AnimationTransitionGraph_0 {
  path=ItemFinalPoseOverride.AnimGraphNode_StateMachine_2.Primary SM.AnimStateTransitionNode_5.AnimationTransitionGraph_0
  nodes [Result(bCanEnterTransition=Get bIsHoverLanding() -> (bIsHoverLanding:bool))]
 }
 Transition {
  path=ItemFinalPoseOverride.AnimGraphNode_StateMachine_2.Primary SM.AnimStateTransitionNode_6.Transition
  nodes [Result(bCanEnterTransition=Get bIsPunchOutro() -> (bIsPunchOutro:bool))]
 }
 AnimationTransitionGraph_0 {
  path=ItemFinalPoseOverride.AnimGraphNode_StateMachine_2.Primary SM.AnimStateTransitionNode_7.AnimationTransitionGraph_0
  nodes [Result(bCanEnterTransition=Get bIsFlightLanding() -> (bIsFlightLanding:bool))]
 }
 AnimationTransitionGraph_0 {path=ItemFinalPoseOverride.AnimGraphNode_StateMachine_2.Primary SM.AnimStateTransitionNode_8.AnimationTransitionGraph_0, nodes [Result()]}
 Transition {
  path=ItemFinalPoseOverride.AnimGraphNode_StateMachine_2.Primary SM.AnimStateTransitionNode_9.Transition
  nodes [Result(bCanEnterTransition=Get bIsNormalFlight())]
 }
 Hover SM {
  entry=Conduit
  path=ItemFinalPoseOverride.AnimGraphNode_StateMachine_4.Hover SM
  ids=id
  nodes [
   To HoverHeatVision() {
    is=title AnimStateAliasNode
    also {aliases [From Charge Transition, From Flight Transition, From Ground Transition, From Sprint Transition, HoverIdle, HoverStart, HoverStop]}
   }
   AnimStateConduitNode_1 = Conduit()
   AnimStateNode_0 = HoverHeatVision Outro() {is=title AnimStateNode}
   AnimStateNode_1 = HoverHeatVision() {is=title AnimStateNode}
   AnimStateNode_15 = HoverStart() {is=title AnimStateNode}
   AnimStateNode_16 = HoverStop() {is=title AnimStateNode}
   AnimStateNode_4 = From Flight Transition()
   AnimStateNode_5 = HoverIdle() {is=title AnimStateNode}
   AnimStateNode_6 = From Ground Transition() {is=title AnimStateNode}
   AnimStateNode_7 = From Sprint Transition()
   AnimStateNode_8 = From Charge Transition() {is=title AnimStateNode}
   AnimStateTransitionNode_0 = AnimStateTransitionNode()
   AnimStateTransitionNode_1 = AnimStateTransitionNode()
   AnimStateTransitionNode_10 = AnimStateTransitionNode() {bAutomaticRuleBasedOnSequencePlayerInState=true}
   AnimStateTransitionNode_11 = AnimStateTransitionNode() {bAutomaticRuleBasedOnSequencePlayerInState=true}
   AnimStateTransitionNode_12 = AnimStateTransitionNode() {bAutomaticRuleBasedOnSequencePlayerInState=true}
   AnimStateTransitionNode_18 = AnimStateTransitionNode() {logicType=TLT_Inertialization}
   AnimStateTransitionNode_2 = AnimStateTransitionNode()
   AnimStateTransitionNode_22 = AnimStateTransitionNode() {crossfadeDuration=0.4}
   AnimStateTransitionNode_23 = AnimStateTransitionNode() {crossfadeDuration=0.6}
   AnimStateTransitionNode_24 = AnimStateTransitionNode() {crossfadeDuration=0.4}
   AnimStateTransitionNode_25 = AnimStateTransitionNode() {crossfadeDuration=0.6}
   AnimStateTransitionNode_29 = AnimStateTransitionNode()
   AnimStateTransitionNode_3 = AnimStateTransitionNode() {bAutomaticRuleBasedOnSequencePlayerInState=true}
   AnimStateTransitionNode_30 = AnimStateTransitionNode()
   AnimStateTransitionNode_31 = AnimStateTransitionNode()
   AnimStateTransitionNode_32 = AnimStateTransitionNode()
   AnimStateTransitionNode_4 = AnimStateTransitionNode() {priorityOrder=10}
   AnimStateTransitionNode_5 = AnimStateTransitionNode() {bAutomaticRuleBasedOnSequencePlayerInState=true}
   AnimStateTransitionNode_6 = AnimStateTransitionNode() {bAutomaticRuleBasedOnSequencePlayerInState=true}
  ]
  links [
   {
    dst=From Charge Transition
    rule=ItemFinalPoseOverride.AnimGraphNode_StateMachine_4.Hover SM.AnimStateTransitionNode_30.AnimationTransitionGraph_2
    src=Conduit
    via=AnimStateTransitionNode_30
   }
   {
    dst=From Flight Transition
    rule=ItemFinalPoseOverride.AnimGraphNode_StateMachine_4.Hover SM.AnimStateTransitionNode_31.Transition
    src=Conduit
    via=AnimStateTransitionNode_31
   }
   {
    dst=From Ground Transition
    rule=ItemFinalPoseOverride.AnimGraphNode_StateMachine_4.Hover SM.AnimStateTransitionNode_32.Transition
    src=Conduit
    via=AnimStateTransitionNode_32
   }
   {
    dst=From Sprint Transition
    rule=ItemFinalPoseOverride.AnimGraphNode_StateMachine_4.Hover SM.AnimStateTransitionNode_29.AnimationTransitionGraph_2
    src=Conduit
    via=AnimStateTransitionNode_29
   }
   {
    dst=HoverIdle
    rule=ItemFinalPoseOverride.AnimGraphNode_StateMachine_4.Hover SM.AnimStateTransitionNode_2.Transition
    src=Conduit
    via=AnimStateTransitionNode_2
   }
   {
    dst=HoverIdle
    rule=ItemFinalPoseOverride.AnimGraphNode_StateMachine_4.Hover SM.AnimStateTransitionNode_4.Transition
    src=Conduit
    via=AnimStateTransitionNode_4
   }
   {
    dst=HoverIdle
    rule=ItemFinalPoseOverride.AnimGraphNode_StateMachine_4.Hover SM.AnimStateTransitionNode_12.Transition
    src=From Charge Transition
    via=AnimStateTransitionNode_12
   }
   {
    dst=HoverIdle
    rule=ItemFinalPoseOverride.AnimGraphNode_StateMachine_4.Hover SM.AnimStateTransitionNode_10.Transition
    src=From Flight Transition
    via=AnimStateTransitionNode_10
   }
   {
    dst=HoverIdle
    rule=ItemFinalPoseOverride.AnimGraphNode_StateMachine_4.Hover SM.AnimStateTransitionNode_3.Transition
    src=From Ground Transition
    via=AnimStateTransitionNode_3
   }
   {
    dst=HoverIdle
    rule=ItemFinalPoseOverride.AnimGraphNode_StateMachine_4.Hover SM.AnimStateTransitionNode_11.Transition
    src=From Sprint Transition
    via=AnimStateTransitionNode_11
   }
   {
    dst=HoverHeatVision Outro
    rule=ItemFinalPoseOverride.AnimGraphNode_StateMachine_4.Hover SM.AnimStateTransitionNode_1.Transition
    src=HoverHeatVision
    via=AnimStateTransitionNode_1
   }
   {
    dst=HoverIdle
    rule=ItemFinalPoseOverride.AnimGraphNode_StateMachine_4.Hover SM.AnimStateTransitionNode_6.Transition
    src=HoverHeatVision Outro
    via=AnimStateTransitionNode_6
   }
   {
    dst=HoverStart
    rule=ItemFinalPoseOverride.AnimGraphNode_StateMachine_4.Hover SM.AnimStateTransitionNode_18.AnimationTransitionGraph_1
    src=HoverIdle
    via=AnimStateTransitionNode_18
   }
   {
    dst=HoverStop
    rule=ItemFinalPoseOverride.AnimGraphNode_StateMachine_4.Hover SM.AnimStateTransitionNode_22.AnimationTransitionGraph_1
    src=HoverIdle
    via=AnimStateTransitionNode_22
   }
   {
    dst=HoverIdle
    rule=ItemFinalPoseOverride.AnimGraphNode_StateMachine_4.Hover SM.AnimStateTransitionNode_5.Transition
    src=HoverStart
    via=AnimStateTransitionNode_5
   }
   {
    dst=HoverStop
    rule=ItemFinalPoseOverride.AnimGraphNode_StateMachine_4.Hover SM.AnimStateTransitionNode_23.AnimationTransitionGraph_2
    src=HoverStart
    via=AnimStateTransitionNode_23
   }
   {
    dst=HoverIdle
    rule=ItemFinalPoseOverride.AnimGraphNode_StateMachine_4.Hover SM.AnimStateTransitionNode_25.AnimationTransitionGraph_1
    src=HoverStop
    via=AnimStateTransitionNode_25
   }
   {
    dst=HoverStart
    rule=ItemFinalPoseOverride.AnimGraphNode_StateMachine_4.Hover SM.AnimStateTransitionNode_24.AnimationTransitionGraph_2
    src=HoverStop
    via=AnimStateTransitionNode_24
   }
   {
    dst=HoverHeatVision
    rule=ItemFinalPoseOverride.AnimGraphNode_StateMachine_4.Hover SM.AnimStateTransitionNode_0.Transition
    src=To HoverHeatVision
    via=AnimStateTransitionNode_0
   }
  ]
 }
 Conduit {path=ItemFinalPoseOverride.AnimGraphNode_StateMachine_4.Hover SM.AnimStateConduitNode_1.Conduit, nodes [Result(bCanEnterTransition=true)]}
 HoverHeatVision Outro {
  path=ItemFinalPoseOverride.AnimGraphNode_StateMachine_4.Hover SM.AnimStateNode_0.HoverHeatVision Outro
  nodes [
   n1 = Sequence Player 'IcicleLarva_HeatVision_InAir_Firing_Outro'() {
    bLoopAnimation=false
    groupName=HeatVisionOutro
    method=SyncGroup
    sequence=/IcicleLarvaGameplay/Assets/Animation/Proto/NewHeatVision/IcicleLarva_HeatVision_InAir_Firing_Outro
   }
   n2 = Sequence Player 'IcicleLarva_HeatVision_Firing_Outro'() {
    bLoopAnimation=false
    groupName=HeatVisionOutro
    method=SyncGroup
    sequence=/IcicleLarvaGameplay/Assets/Animation/Proto/NewHeatVision/IcicleLarva_HeatVision_Firing_Outro
   }
   Output Animation Pose(Result=Blend Poses by bool(BlendPose_0=@n1, BlendPose_1=@n2, bActiveValue=Get bIsHover()) {blendType=Linear})
   Get bIsHeatVisionFlight() -> (bIsHeatVisionFlight:bool)
  ]
 }
 HoverHeatVision {
  path=ItemFinalPoseOverride.AnimGraphNode_StateMachine_4.Hover SM.AnimStateNode_1.HoverHeatVision
  nodes [
   AnimGraphNode_StateMachine_0 = HoverHeatVision SM() {is=title AnimGraphNode_StateMachine}
   Output Animation Pose(Result=Modify Curve!(SourcePose=@AnimGraphNode_StateMachine_0, CurveValues_0=0, Alpha=1) -> (Pose) {curveNames [HoverPivotAlpha], curveValues [0]})
  ]
 }
 HoverHeatVision SM {
  entry=Pass Through
  path=ItemFinalPoseOverride.AnimGraphNode_StateMachine_4.Hover SM.AnimStateNode_1.HoverHeatVision.AnimGraphNode_StateMachine_0.HoverHeatVision SM
  ids=id
  nodes [
   AnimStateNode_0 = HeatVision Outro() {is=title AnimStateNode}
   AnimStateNode_1 = HeatVision Intro() {is=title AnimStateNode}
   AnimStateNode_2 = HeatVision Loop() {is=title AnimStateNode}
   AnimStateNode_3 = Pass Through() {is=title AnimStateNode}
   AnimStateTransitionNode_15 = AnimStateTransitionNode()
   AnimStateTransitionNode_16 = AnimStateTransitionNode() {bAutomaticRuleBasedOnSequencePlayerInState=true}
   AnimStateTransitionNode_3 = AnimStateTransitionNode() {bAutomaticRuleBasedOnSequencePlayerInState=true}
   AnimStateTransitionNode_4 = AnimStateTransitionNode()
  ]
  links [
   {
    dst=HeatVision Loop
    rule=ItemFinalPoseOverride.AnimGraphNode_StateMachine_4.Hover SM.AnimStateNode_1.HoverHeatVision.AnimGraphNode_StateMachine_0.HoverHeatVision SM.AnimStateTransitionNode_3.AnimationTransitionGraph_0
    src=HeatVision Intro
    via=AnimStateTransitionNode_3
   }
   {
    dst=HeatVision Outro
    rule=ItemFinalPoseOverride.AnimGraphNode_StateMachine_4.Hover SM.AnimStateNode_1.HoverHeatVision.AnimGraphNode_StateMachine_0.HoverHeatVision SM.AnimStateTransitionNode_4.AnimationTransitionGraph_0
    src=HeatVision Loop
    via=AnimStateTransitionNode_4
   }
   {
    dst=Pass Through
    rule=ItemFinalPoseOverride.AnimGraphNode_StateMachine_4.Hover SM.AnimStateNode_1.HoverHeatVision.AnimGraphNode_StateMachine_0.HoverHeatVision SM.AnimStateTransitionNode_16.AnimationTransitionGraph_0
    src=HeatVision Outro
    via=AnimStateTransitionNode_16
   }
   {
    dst=HeatVision Intro
    rule=ItemFinalPoseOverride.AnimGraphNode_StateMachine_4.Hover SM.AnimStateNode_1.HoverHeatVision.AnimGraphNode_StateMachine_0.HoverHeatVision SM.AnimStateTransitionNode_15.AnimationTransitionGraph_0
    src=Pass Through
    via=AnimStateTransitionNode_15
   }
  ]
 }
 HeatVision Outro {
  path=ItemFinalPoseOverride.AnimGraphNode_StateMachine_4.Hover SM.AnimStateNode_1.HoverHeatVision.AnimGraphNode_StateMachine_0.HoverHeatVision SM.AnimStateNode_0.HeatVision Outro
  nodes [
   n1 = Sequence Player 'IcicleLarva_HeatVision_InAir_Firing_Outro'() {
    bLoopAnimation=false
    groupName=HeatVisionOutro
    method=SyncGroup
    sequence=&IcicleLarva_HeatVision_InAir_Firing_Outro
   }
   n2 = Sequence Player 'IcicleLarva_HeatVision_Firing_Outro'() {
    bLoopAnimation=false
    groupName=HeatVisionOutro
    method=SyncGroup
    sequence=&IcicleLarva_HeatVision_Firing_Outro
   }
   Output Animation Pose(Result=Blend Poses by bool(BlendPose_0=@n1, BlendPose_1=@n2, bActiveValue=Get bIsHover()) {blendType=Linear}) {
    node {stateEntryFunction {functionName=HeatVisionOutro}}
    stateEntryFunction {bSelfContext=true, memberGuid=#26, memberName=HeatVisionOutro}
   }
   Get bIsHeatVisionFlight()
  ]
 }
 HeatVision Intro {
  path=ItemFinalPoseOverride.AnimGraphNode_StateMachine_4.Hover SM.AnimStateNode_1.HoverHeatVision.AnimGraphNode_StateMachine_0.HoverHeatVision SM.AnimStateNode_1.HeatVision Intro
  nodes [
   n1 = Sequence Player 'IcicleLarva_HeatVision_InAir_Firing_Intro'() {
    bLoopAnimation=false
    groupName=HeatVision
    method=SyncGroup
    sequence=/IcicleLarvaGameplay/Assets/Animation/Proto/NewHeatVision/IcicleLarva_HeatVision_InAir_Firing_Intro
   }
   n2 = Sequence Player 'IcicleLarva_HeatVision_Firing_Intro'() {
    bLoopAnimation=false
    groupName=HeatVision
    method=SyncGroup
    sequence=/IcicleLarvaGameplay/Assets/Animation/Proto/NewHeatVision/IcicleLarva_HeatVision_Firing_Intro
   }
   Output Animation Pose(Result=Blend Poses by bool(BlendPose_0=@n1, BlendPose_1=@n2, bActiveValue=Get bIsHeatVisionFlight()) {blendType=Linear}) {
    node {stateEntryFunction {functionName=OnStateEntry}}
    stateEntryFunction {bSelfContext=true, memberGuid=#27, memberName=HeatVisionIntro}
   }
  ]
 }
 HeatVision Loop {
  path=ItemFinalPoseOverride.AnimGraphNode_StateMachine_4.Hover SM.AnimStateNode_1.HoverHeatVision.AnimGraphNode_StateMachine_0.HoverHeatVision SM.AnimStateNode_2.HeatVision Loop
  nodes [Output Animation Pose(Result=Blend Poses by bool(
     BlendPose_0=Sequence Player 'IcicleLarva_HeatVision_InAir_Firing_Loop'() {bLoopAnimation=false, sequence=/IcicleLarvaGameplay/Assets/Animation/Proto/NewHeatVision/IcicleLarva_HeatVision_InAir_Firing_Loop}
     BlendPose_1=Sequence Player 'IcicleLarva_HeatVision_Firing_Loop'() {bLoopAnimation=false, sequence=/IcicleLarvaGameplay/Assets/Animation/Proto/NewHeatVision/IcicleLarva_HeatVision_Firing_Loop}
     bActiveValue=Get bIsHeatVisionFlight()
    ) {blendType=Linear}) {
    node {stateEntryFunction {functionName=HeatVisionLoop}}
    stateEntryFunction {bSelfContext=true, memberGuid=#28, memberName=HeatVisionLoop}
   }]
 }
 Pass Through {
  path=ItemFinalPoseOverride.AnimGraphNode_StateMachine_4.Hover SM.AnimStateNode_1.HoverHeatVision.AnimGraphNode_StateMachine_0.HoverHeatVision SM.AnimStateNode_3.Pass Through
  nodes [Output Animation Pose(Result=Use cached pose 'FinalPose'())]
 }
 AnimationTransitionGraph_0 {
  path=ItemFinalPoseOverride.AnimGraphNode_StateMachine_4.Hover SM.AnimStateNode_1.HoverHeatVision.AnimGraphNode_StateMachine_0.HoverHeatVision SM.AnimStateTransitionNode_15.AnimationTransitionGraph_0
  nodes [Result(bCanEnterTransition=Get bIsHeatVision() -> (bIsHeatVision:bool))]
 }
 AnimationTransitionGraph_0 {
  path=ItemFinalPoseOverride.AnimGraphNode_StateMachine_4.Hover SM.AnimStateNode_1.HoverHeatVision.AnimGraphNode_StateMachine_0.HoverHeatVision SM.AnimStateTransitionNode_16.AnimationTransitionGraph_0
  nodes [Result()]
 }
 AnimationTransitionGraph_0 {
  path=ItemFinalPoseOverride.AnimGraphNode_StateMachine_4.Hover SM.AnimStateNode_1.HoverHeatVision.AnimGraphNode_StateMachine_0.HoverHeatVision SM.AnimStateTransitionNode_3.AnimationTransitionGraph_0
  nodes [Result()]
 }
 AnimationTransitionGraph_0 {
  path=ItemFinalPoseOverride.AnimGraphNode_StateMachine_4.Hover SM.AnimStateNode_1.HoverHeatVision.AnimGraphNode_StateMachine_0.HoverHeatVision SM.AnimStateTransitionNode_4.AnimationTransitionGraph_0
  nodes [Result(bCanEnterTransition=NOT Boolean(A=Get bIsHeatVision()))]
 }
 HoverStart {
  path=ItemFinalPoseOverride.AnimGraphNode_StateMachine_4.Hover SM.AnimStateNode_15.HoverStart
  nodes [
   n1 = Blendspace Player 'IcicleLarva_Hover_Start_BS'!(X=0, Y=0, PlayRate:float=1.25) -> (Pose) {bLoop=false, blendSpace=/IcicleLarvaGameplay/Assets/Animation/Flight/Hover/Transitions/IcicleLarva_Hover_Start_BS}
   n2 = Sequence Player 'IcicleLarva_Hover_Idle'() {sequence=/IcicleLarvaGameplay/Assets/Animation/Flight/Hover/IcicleLarva_Hover_Idle}
   Output Animation Pose(Result=Modify Curve(SourcePose=Two Way Blend(A=@n1, B=@n2, Alpha=0.75) {blendNode {alphaBoolBlend {blendOption=Linear}}}, CurveValues_0=0.5, Alpha=1) {curveNames [IcicleLarvaLeanAlphaCurve], curveValues [0]})
  ]
 }
 HoverStop {
  path=ItemFinalPoseOverride.AnimGraphNode_StateMachine_4.Hover SM.AnimStateNode_16.HoverStop
  nodes [Output Animation Pose(Result=Modify Curve(
     SourcePose=Blendspace Player 'IcicleLarva_Hover_Stop_BS'!(X=0, Y=0, PlayRate:float=1.25) -> (Pose) {bLoop=false, blendSpace=/IcicleLarvaGameplay/Assets/Animation/Flight/Hover/Transitions/IcicleLarva_Hover_Stop_BS}
     CurveValues_0=1
     Alpha=1
    ) {curveNames [IcicleLarvaLeanAlphaCurve], curveValues [0]})]
 }
 From Flight Transition {
  path=ItemFinalPoseOverride.AnimGraphNode_StateMachine_4.Hover SM.AnimStateNode_4.From Flight Transition
  nodes [Output Animation Pose(Result=Modify Curve(
     SourcePose=Sequence Player 'IcicleLarva_Hover_Pose_Base'() {bLoopAnimation=false, sequence=/IcicleLarvaGameplay/Assets/Animation/Flight/Hover/IcicleLarva_Hover_Pose_Base}
     CurveValues_0=0
     Alpha=1
    ) {curveNames [HoverPivotAlpha], curveValues [0]}) {
    node {stateEntryFunction {functionName=FromFlightFalse}}
    stateEntryFunction {bSelfContext=true, memberGuid=#16, memberName=FromFlightFalse}
   }]
 }
 HoverIdle {
  path=ItemFinalPoseOverride.AnimGraphNode_StateMachine_4.Hover SM.AnimStateNode_5.HoverIdle
  nodes [Output Animation Pose(Result=Modify Curve(
     SourcePose=Sequence Player 'IcicleLarva_Hover_Idle'() {sequence=&IcicleLarva_Hover_Idle}
     CurveValues_0=1
     Alpha=1
    ) {curveNames [IcicleLarvaLeanAlphaCurve], curveValues [0]}) {
    becomeRelevantFunction {bSelfContext=true, memberGuid=#29, memberName=FromIntroFalse}
    node {becomeRelevantFunction {functionName=FromIntroFalse}}
   }]
 }
 From Ground Transition {
  path=ItemFinalPoseOverride.AnimGraphNode_StateMachine_4.Hover SM.AnimStateNode_6.From Ground Transition
  nodes [Output Animation Pose(Result=Modify Curve(
     SourcePose=Sequence Player 'IcicleLarva_Hover_Pose_Base'() {bLoopAnimation=false, sequence=&IcicleLarva_Hover_Pose_Base}
     CurveValues_0=0
     Alpha=1
    ) {curveNames [HoverPivotAlpha], curveValues [0]})]
 }
 From Sprint Transition {
  path=ItemFinalPoseOverride.AnimGraphNode_StateMachine_4.Hover SM.AnimStateNode_7.From Sprint Transition
  nodes [Output Animation Pose(Result=Modify Curve(
     SourcePose=Sequence Player 'IcicleLarva_Flight_Transition_Stop'() {bLoopAnimation=false, sequence=/IcicleLarvaGameplay/Assets/Animation/Flight/IcicleLarva_Flight_Transition_Stop}
     CurveValues_0=0
     Alpha=1
    ) {curveNames [HoverPivotAlpha], curveValues [0]}) {
    becomeRelevantFunction {bSelfContext=true, memberGuid=#17, memberName=FromSprintFalse}
    node {becomeRelevantFunction {functionName=FromSprintFalse}}
   }]
 }
 From Charge Transition {
  path=ItemFinalPoseOverride.AnimGraphNode_StateMachine_4.Hover SM.AnimStateNode_8.From Charge Transition
  nodes [Output Animation Pose(Result=Modify Curve(
     SourcePose=Sequence Player 'IcicleLarva_Flight_Transition_ChargeFlyToHovering'() {bLoopAnimation=false, sequence=/IcicleLarvaGameplay/Assets/Animation/Proto/IcicleLarva_Flight_Transition_ChargeFlyToHovering}
     CurveValues_0=0
     Alpha=1
    ) {curveNames [HoverPivotAlpha], curveValues [0]}) {
    becomeRelevantFunction {bSelfContext=true, memberGuid=#22, memberName=FromChargeFalse}
    node {becomeRelevantFunction {functionName=FromChargeFalse}}
   }]
 }
 Transition {
  path=ItemFinalPoseOverride.AnimGraphNode_StateMachine_4.Hover SM.AnimStateTransitionNode_0.Transition
  nodes [Result(bCanEnterTransition=Get bIsHeatVisionFlight())]
 }
 Transition {
  path=ItemFinalPoseOverride.AnimGraphNode_StateMachine_4.Hover SM.AnimStateTransitionNode_1.Transition
  nodes [Result(bCanEnterTransition=NOT Boolean(A=Get bIsHeatVisionFlight()))]
 }
 Transition {path=ItemFinalPoseOverride.AnimGraphNode_StateMachine_4.Hover SM.AnimStateTransitionNode_10.Transition, nodes [Result()]}
 Transition {path=ItemFinalPoseOverride.AnimGraphNode_StateMachine_4.Hover SM.AnimStateTransitionNode_11.Transition, nodes [Result()]}
 Transition {path=ItemFinalPoseOverride.AnimGraphNode_StateMachine_4.Hover SM.AnimStateTransitionNode_12.Transition, nodes [Result()]}
 AnimationTransitionGraph_1 {
  path=ItemFinalPoseOverride.AnimGraphNode_StateMachine_4.Hover SM.AnimStateTransitionNode_18.AnimationTransitionGraph_1
  nodes [Result(bCanEnterTransition=Get bIsHoverStart() -> (bIsHoverStart:bool))]
 }
 Transition {
  path=ItemFinalPoseOverride.AnimGraphNode_StateMachine_4.Hover SM.AnimStateTransitionNode_2.Transition
  nodes [Result(bCanEnterTransition=Get bIsFromIntro() -> (bIsFromIntro:bool))]
 }
 AnimationTransitionGraph_1 {
  path=ItemFinalPoseOverride.AnimGraphNode_StateMachine_4.Hover SM.AnimStateTransitionNode_22.AnimationTransitionGraph_1
  nodes [Result(bCanEnterTransition=Get bHoverIdleToHoverStopTransition() -> (bHoverIdleToHoverStopTransition:bool))]
 }
 AnimationTransitionGraph_2 {
  path=ItemFinalPoseOverride.AnimGraphNode_StateMachine_4.Hover SM.AnimStateTransitionNode_23.AnimationTransitionGraph_2
  nodes [Result(bCanEnterTransition=Get bHoverStartToHoverStopTransition() -> (bHoverStartToHoverStopTransition:bool))]
 }
 AnimationTransitionGraph_2 {
  path=ItemFinalPoseOverride.AnimGraphNode_StateMachine_4.Hover SM.AnimStateTransitionNode_24.AnimationTransitionGraph_2
  nodes [Result(bCanEnterTransition=Get bHoverStopToHoverStartTransition() -> (bHoverStopToHoverStartTransition:bool))]
 }
 AnimationTransitionGraph_1 {
  path=ItemFinalPoseOverride.AnimGraphNode_StateMachine_4.Hover SM.AnimStateTransitionNode_25.AnimationTransitionGraph_1
  nodes [Result(bCanEnterTransition=NOT Boolean(A=Get bIsHoverStop() -> (bIsHoverStop:bool)))]
 }
 AnimationTransitionGraph_2 {
  path=ItemFinalPoseOverride.AnimGraphNode_StateMachine_4.Hover SM.AnimStateTransitionNode_29.AnimationTransitionGraph_2
  nodes [Result(bCanEnterTransition=Get bIsFromSprint())]
 }
 Transition {path=ItemFinalPoseOverride.AnimGraphNode_StateMachine_4.Hover SM.AnimStateTransitionNode_3.Transition, nodes [Result()]}
 AnimationTransitionGraph_2 {
  path=ItemFinalPoseOverride.AnimGraphNode_StateMachine_4.Hover SM.AnimStateTransitionNode_30.AnimationTransitionGraph_2
  nodes [Result(bCanEnterTransition=Get bIsFromCharge() -> (bIsFromCharge:bool))]
 }
 Transition {
  path=ItemFinalPoseOverride.AnimGraphNode_StateMachine_4.Hover SM.AnimStateTransitionNode_31.Transition
  nodes [Result(bCanEnterTransition=Get bIsFromFlight())]
 }
 Transition {
  path=ItemFinalPoseOverride.AnimGraphNode_StateMachine_4.Hover SM.AnimStateTransitionNode_32.Transition
  nodes [Result(bCanEnterTransition=Get bIsAscend())]
 }
 Transition {
  path=ItemFinalPoseOverride.AnimGraphNode_StateMachine_4.Hover SM.AnimStateTransitionNode_4.Transition
  nodes [Result(bCanEnterTransition=true)]
 }
 Transition {path=ItemFinalPoseOverride.AnimGraphNode_StateMachine_4.Hover SM.AnimStateTransitionNode_5.Transition, nodes [Result()]}
 Transition {path=ItemFinalPoseOverride.AnimGraphNode_StateMachine_4.Hover SM.AnimStateTransitionNode_6.Transition, nodes [Result()]}
 NormalFlight SM {
  entry=Conduit
  path=ItemFinalPoseOverride.AnimGraphNode_StateMachine_5.NormalFlight SM
  ids=id
  nodes [
   AnimStateConduitNode_1 = Conduit()
   AnimStateNode_0 = From Charge Transition()
   AnimStateNode_1 = Normal Flight() {is=title AnimStateNode}
   AnimStateNode_2 = Safety Transition()
   AnimStateNode_3 = From Sprint Transition()
   AnimStateNode_4 = Up Pitch React() {is=title AnimStateNode}
   AnimStateNode_5 = Down Pitch React() {is=title AnimStateNode}
   AnimStateNode_6 = From Hover Transition()
   AnimStateNode_7 = Yaw React() {is=title AnimStateNode}
   AnimStateTransitionNode_0 = AnimStateTransitionNode() {crossfadeDuration=0.1}
   AnimStateTransitionNode_1 = AnimStateTransitionNode() {crossfadeDuration=0.1}
   AnimStateTransitionNode_10 = AnimStateTransitionNode() {bAutomaticRuleBasedOnSequencePlayerInState=true}
   AnimStateTransitionNode_2 = AnimStateTransitionNode() {bAutomaticRuleBasedOnSequencePlayerInState=true}
   AnimStateTransitionNode_26 = AnimStateTransitionNode() {crossfadeDuration=0.01}
   AnimStateTransitionNode_29 = AnimStateTransitionNode() {crossfadeDuration=0.01}
   AnimStateTransitionNode_3 = AnimStateTransitionNode() {bAutomaticRuleBasedOnSequencePlayerInState=true}
   AnimStateTransitionNode_30 = AnimStateTransitionNode()
   AnimStateTransitionNode_4 = AnimStateTransitionNode() {bAutomaticRuleBasedOnSequencePlayerInState=true}
   AnimStateTransitionNode_5 = AnimStateTransitionNode() {bAutomaticRuleBasedOnSequencePlayerInState=true}
   AnimStateTransitionNode_6 = AnimStateTransitionNode() {priorityOrder=10}
   AnimStateTransitionNode_7 = AnimStateTransitionNode() {bAutomaticRuleBasedOnSequencePlayerInState=true}
   AnimStateTransitionNode_8 = AnimStateTransitionNode() {bDisabled=true}
   AnimStateTransitionNode_9 = AnimStateTransitionNode() {bAutomaticRuleBasedOnSequencePlayerInState=true, crossfadeDuration=1}
  ]
  links [
   {
    dst=From Charge Transition
    rule=ItemFinalPoseOverride.AnimGraphNode_StateMachine_5.NormalFlight SM.AnimStateTransitionNode_30.AnimationTransitionGraph_2
    src=Conduit
    via=AnimStateTransitionNode_30
   }
   {
    dst=From Hover Transition
    rule=ItemFinalPoseOverride.AnimGraphNode_StateMachine_5.NormalFlight SM.AnimStateTransitionNode_26.AnimationTransitionGraph_2
    src=Conduit
    via=AnimStateTransitionNode_26
   }
   {
    dst=From Sprint Transition
    rule=ItemFinalPoseOverride.AnimGraphNode_StateMachine_5.NormalFlight SM.AnimStateTransitionNode_29.AnimationTransitionGraph_2
    src=Conduit
    via=AnimStateTransitionNode_29
   }
   {
    dst=Safety Transition
    rule=ItemFinalPoseOverride.AnimGraphNode_StateMachine_5.NormalFlight SM.AnimStateTransitionNode_6.Transition
    src=Conduit
    via=AnimStateTransitionNode_6
   }
   {
    dst=Normal Flight
    rule=ItemFinalPoseOverride.AnimGraphNode_StateMachine_5.NormalFlight SM.AnimStateTransitionNode_3.AnimationTransitionGraph_0
    src=Down Pitch React
    via=AnimStateTransitionNode_3
   }
   {
    dst=Normal Flight
    rule=ItemFinalPoseOverride.AnimGraphNode_StateMachine_5.NormalFlight SM.AnimStateTransitionNode_7.AnimationTransitionGraph_1
    src=From Charge Transition
    via=AnimStateTransitionNode_7
   }
   {
    dst=Normal Flight
    rule=ItemFinalPoseOverride.AnimGraphNode_StateMachine_5.NormalFlight SM.AnimStateTransitionNode_2.AnimationTransitionGraph_1
    src=From Hover Transition
    via=AnimStateTransitionNode_2
   }
   {
    dst=Normal Flight
    rule=ItemFinalPoseOverride.AnimGraphNode_StateMachine_5.NormalFlight SM.AnimStateTransitionNode_4.AnimationTransitionGraph_1
    src=From Sprint Transition
    via=AnimStateTransitionNode_4
   }
   {
    dst=Down Pitch React
    rule=ItemFinalPoseOverride.AnimGraphNode_StateMachine_5.NormalFlight SM.AnimStateTransitionNode_0.AnimationTransitionGraph_0
    src=Normal Flight
    via=AnimStateTransitionNode_0
   }
   {
    dst=Up Pitch React
    rule=ItemFinalPoseOverride.AnimGraphNode_StateMachine_5.NormalFlight SM.AnimStateTransitionNode_1.AnimationTransitionGraph_0
    src=Normal Flight
    via=AnimStateTransitionNode_1
   }
   {
    dst=Yaw React
    rule=ItemFinalPoseOverride.AnimGraphNode_StateMachine_5.NormalFlight SM.AnimStateTransitionNode_8.AnimationTransitionGraph_0
    src=Normal Flight
    via=AnimStateTransitionNode_8
   }
   {
    dst=Normal Flight
    rule=ItemFinalPoseOverride.AnimGraphNode_StateMachine_5.NormalFlight SM.AnimStateTransitionNode_10.Transition
    src=Safety Transition
    via=AnimStateTransitionNode_10
   }
   {
    dst=Normal Flight
    rule=ItemFinalPoseOverride.AnimGraphNode_StateMachine_5.NormalFlight SM.AnimStateTransitionNode_5.AnimationTransitionGraph_0
    src=Up Pitch React
    via=AnimStateTransitionNode_5
   }
   {
    dst=Normal Flight
    rule=ItemFinalPoseOverride.AnimGraphNode_StateMachine_5.NormalFlight SM.AnimStateTransitionNode_9.AnimationTransitionGraph_0
    src=Yaw React
    via=AnimStateTransitionNode_9
   }
  ]
 }
 Conduit {
  path=ItemFinalPoseOverride.AnimGraphNode_StateMachine_5.NormalFlight SM.AnimStateConduitNode_1.Conduit
  nodes [Result(bCanEnterTransition=true)]
 }
 From Charge Transition {
  path=ItemFinalPoseOverride.AnimGraphNode_StateMachine_5.NormalFlight SM.AnimStateNode_0.From Charge Transition
  nodes [
   n1 = Sequence Player 'IcicleLarva_Flight_Transition_TwoHandedToHandsBackFly'() {
    bLoopAnimation=false
    sequence=/IcicleLarvaGameplay/Assets/Animation/Flight/HandsBackFlight/IcicleLarva_Flight_Transition_TwoHandedToHandsBackFly
   }
   Output Animation Pose(Result=@n1) {
    becomeRelevantFunction {bSelfContext=true, memberGuid=#22, memberName=FromChargeFalse}
    node {becomeRelevantFunction {functionName=FromChargeFalse}}
   }
  ]
 }
 Normal Flight {
  path=ItemFinalPoseOverride.AnimGraphNode_StateMachine_5.NormalFlight SM.AnimStateNode_1.Normal Flight
  nodes [Output Animation Pose(Result=Sequence Player 'IcicleLarva_Flight_HandsBack_Idle_Loop'() {sequence=&IcicleLarva_Flight_HandsBack_Idle_Loop}) {
    node {stateEntryFunction {functionName=ResetFlightYawReactReaction}}
    stateEntryFunction {bSelfContext=true, memberGuid=#30, memberName=ResetFlightYawReactReaction}
   }]
 }
 Safety Transition {
  path=ItemFinalPoseOverride.AnimGraphNode_StateMachine_5.NormalFlight SM.AnimStateNode_2.Safety Transition
  nodes [Output Animation Pose(Result=Two Way Blend(
     A=Sequence Player 'IcicleLarva_Flight_HandsBack_Idle_Loop'() {sequence=&IcicleLarva_Flight_HandsBack_Idle_Loop}
     B=Sequence Player 'IcicleLarva_HovertoFlight_Transition'() {bLoopAnimation=false, sequence=/IcicleLarvaGameplay/Assets/Animation/Flight/Hover/IcicleLarva_HovertoFlight_Transition}
     Alpha=0.6
    ) {blendNode {alphaBoolBlend {blendOption=Linear}}}) {
    becomeRelevantFunction {bSelfContext=true, memberGuid=#15, memberName=FromHoverFalse}
    node {becomeRelevantFunction {functionName=FromHoverFalse}}
   }]
 }
 From Sprint Transition {
  path=ItemFinalPoseOverride.AnimGraphNode_StateMachine_5.NormalFlight SM.AnimStateNode_3.From Sprint Transition
  nodes [Output Animation Pose(Result=Modify Curve(
     SourcePose=Sequence Player 'IcicleLarva_Flight_Transition_SprintFlight_to_Flight'() {bLoopAnimation=false, sequence=/IcicleLarvaGameplay/Assets/Animation/Flight/IcicleLarva_Flight_Transition_SprintFlight_to_Flight}
     CurveValues_0=1
     Alpha=0
    ) {curveNames [DisableFlightYawAdditive], curveValues [0]}) {
    becomeRelevantFunction {bSelfContext=true, memberGuid=#17, memberName=FromSprintFalse}
    node {becomeRelevantFunction {functionName=FromSprintFalse}}
   }]
 }
 Up Pitch React {
  path=ItemFinalPoseOverride.AnimGraphNode_StateMachine_5.NormalFlight SM.AnimStateNode_4.Up Pitch React
  nodes [
   n1 = Sequence Player 'Icicle_Larva_Flight_HandsBack_Trans_toUpFlying'!(StartPosition:float=0.1) -> (Pose) {bLoopAnimation=false, sequence=/IcicleLarvaGameplay/Assets/Animation/Flight/HandsBackFlight/Icicle_Larva_Flight_HandsBack_Trans_toUpFlying}
   Output Animation Pose(Result=@n1)
  ]
 }
 Down Pitch React {
  path=ItemFinalPoseOverride.AnimGraphNode_StateMachine_5.NormalFlight SM.AnimStateNode_5.Down Pitch React
  nodes [
   n1 = Sequence Player 'Icicle_Larva_Flight_HandsBack_Trans_toDownFlying'!(StartPosition:float=0.1) -> (Pose) {bLoopAnimation=false, sequence=/IcicleLarvaGameplay/Assets/Animation/Flight/HandsBackFlight/Icicle_Larva_Flight_HandsBack_Trans_toDownFlying}
   Output Animation Pose(Result=@n1)
  ]
 }
 From Hover Transition {
  path=ItemFinalPoseOverride.AnimGraphNode_StateMachine_5.NormalFlight SM.AnimStateNode_6.From Hover Transition
  nodes [
   n1 = Sequence Player 'IcicleLarva_HovertoFlight_Transition'() {bLoopAnimation=false, sequence=&IcicleLarva_HovertoFlight_Transition}
   n2 = Sequence Player 'IcicleLarva_Flight_HandsBack_Idle_Loop'() {sequence=&IcicleLarva_Flight_HandsBack_Idle_Loop}
   Output Animation Pose(Result=Modify Curve(SourcePose=Two Way Blend(A=@n2, B=@n1, Alpha=0.8) {blendNode {alphaBoolBlend {blendOption=Linear}}}, CurveValues_0=0, Alpha=1) {curveNames [HoverPivotAlpha], curveValues [0]}) {
    becomeRelevantFunction {bSelfContext=true, memberGuid=#15, memberName=FromHoverFalse}
    node {becomeRelevantFunction {functionName=FromHoverFalse}}
   }
  ]
 }
 Yaw React {
  path=ItemFinalPoseOverride.AnimGraphNode_StateMachine_5.NormalFlight SM.AnimStateNode_7.Yaw React
  nodes [Output Animation Pose(Result=Two Way Blend(
     A=Sequence Player 'IcicleLarva_Flight_HandsBack_Idle_Loop'() {bLoopAnimation=false, sequence=&IcicleLarva_Flight_HandsBack_Idle_Loop}
     B=Sequence Player 'IcicleLarva_HovertoFlight_Transition'() {bLoopAnimation=false, sequence=&IcicleLarva_HovertoFlight_Transition}
     Alpha=0.25
    ) {blendNode {alphaBoolBlend {blendOption=Linear}}}) {
    node {stateExitFunction {functionName=ResetFlightRightYawReactReaction}}
    stateExitFunction {bSelfContext=true, memberGuid=#30, memberName=ResetFlightYawReactReaction}
   }]
 }
 AnimationTransitionGraph_0 {
  path=ItemFinalPoseOverride.AnimGraphNode_StateMachine_5.NormalFlight SM.AnimStateTransitionNode_0.AnimationTransitionGraph_0
  nodes [Result(bCanEnterTransition=Get bIsFlightDownPitchReaction() -> (bIsFlightDownPitchReaction:bool))]
 }
 AnimationTransitionGraph_0 {
  path=ItemFinalPoseOverride.AnimGraphNode_StateMachine_5.NormalFlight SM.AnimStateTransitionNode_1.AnimationTransitionGraph_0
  nodes [Result(bCanEnterTransition=Get bIsFlightUpPitchReaction() -> (bIsFlightUpPitchReaction:bool))]
 }
 Transition {path=ItemFinalPoseOverride.AnimGraphNode_StateMachine_5.NormalFlight SM.AnimStateTransitionNode_10.Transition, nodes [Result()]}
 AnimationTransitionGraph_1 {
  path=ItemFinalPoseOverride.AnimGraphNode_StateMachine_5.NormalFlight SM.AnimStateTransitionNode_2.AnimationTransitionGraph_1
  nodes [Result()]
 }
 AnimationTransitionGraph_2 {
  path=ItemFinalPoseOverride.AnimGraphNode_StateMachine_5.NormalFlight SM.AnimStateTransitionNode_26.AnimationTransitionGraph_2
  nodes [Result(bCanEnterTransition=Get bIsFromHover())]
 }
 AnimationTransitionGraph_2 {
  path=ItemFinalPoseOverride.AnimGraphNode_StateMachine_5.NormalFlight SM.AnimStateTransitionNode_29.AnimationTransitionGraph_2
  nodes [Result(bCanEnterTransition=Get bIsFromSprint())]
 }
 AnimationTransitionGraph_0 {
  path=ItemFinalPoseOverride.AnimGraphNode_StateMachine_5.NormalFlight SM.AnimStateTransitionNode_3.AnimationTransitionGraph_0
  nodes [Result()]
 }
 AnimationTransitionGraph_2 {
  path=ItemFinalPoseOverride.AnimGraphNode_StateMachine_5.NormalFlight SM.AnimStateTransitionNode_30.AnimationTransitionGraph_2
  nodes [Result(bCanEnterTransition=Get bIsFromCharge())]
 }
 AnimationTransitionGraph_1 {
  path=ItemFinalPoseOverride.AnimGraphNode_StateMachine_5.NormalFlight SM.AnimStateTransitionNode_4.AnimationTransitionGraph_1
  nodes [Result()]
 }
 AnimationTransitionGraph_0 {
  path=ItemFinalPoseOverride.AnimGraphNode_StateMachine_5.NormalFlight SM.AnimStateTransitionNode_5.AnimationTransitionGraph_0
  nodes [Result()]
 }
 Transition {
  path=ItemFinalPoseOverride.AnimGraphNode_StateMachine_5.NormalFlight SM.AnimStateTransitionNode_6.Transition
  nodes [Result(bCanEnterTransition=true)]
 }
 AnimationTransitionGraph_1 {
  path=ItemFinalPoseOverride.AnimGraphNode_StateMachine_5.NormalFlight SM.AnimStateTransitionNode_7.AnimationTransitionGraph_1
  nodes [Result()]
 }
 AnimationTransitionGraph_0 {
  path=ItemFinalPoseOverride.AnimGraphNode_StateMachine_5.NormalFlight SM.AnimStateTransitionNode_8.AnimationTransitionGraph_0
  nodes [Result(bCanEnterTransition=Get bIsFlightYawReaction() -> (bIsFlightYawReaction:bool))]
 }
 AnimationTransitionGraph_0 {
  path=ItemFinalPoseOverride.AnimGraphNode_StateMachine_5.NormalFlight SM.AnimStateTransitionNode_9.AnimationTransitionGraph_0
  nodes [Result(bCanEnterTransition=true)]
 }
 SprintFlight SM {
  entry=Conduit
  path=ItemFinalPoseOverride.AnimGraphNode_StateMachine_6.SprintFlight SM
  ids=id
  nodes [
   AnimStateConduitNode_1 = Conduit()
   AnimStateNode_0 = Sprint Flight() {is=title AnimStateNode}
   AnimStateNode_1 = From Hover Transition()
   AnimStateNode_2 = Safety Transition()
   AnimStateNode_3 = From Flight Transition()
   AnimStateNode_4 = Up Pitch React()
   AnimStateNode_5 = Down Pitch React()
   AnimStateNode_8 = From Charge Transition()
   AnimStateTransitionNode_0 = AnimStateTransitionNode() {priorityOrder=10}
   AnimStateTransitionNode_11 = AnimStateTransitionNode()
   AnimStateTransitionNode_12 = AnimStateTransitionNode()
   AnimStateTransitionNode_13 = AnimStateTransitionNode() {bAutomaticRuleBasedOnSequencePlayerInState=true}
   AnimStateTransitionNode_14 = AnimStateTransitionNode() {bAutomaticRuleBasedOnSequencePlayerInState=true}
   AnimStateTransitionNode_2 = AnimStateTransitionNode() {bAutomaticRuleBasedOnSequencePlayerInState=true}
   AnimStateTransitionNode_26 = AnimStateTransitionNode() {crossfadeDuration=0.01}
   AnimStateTransitionNode_29 = AnimStateTransitionNode() {crossfadeDuration=0.01}
   AnimStateTransitionNode_3 = AnimStateTransitionNode() {bAutomaticRuleBasedOnSequencePlayerInState=true}
   AnimStateTransitionNode_30 = AnimStateTransitionNode()
   AnimStateTransitionNode_4 = AnimStateTransitionNode() {bAutomaticRuleBasedOnSequencePlayerInState=true, crossfadeDuration=0.01}
   AnimStateTransitionNode_7 = AnimStateTransitionNode() {bAutomaticRuleBasedOnSequencePlayerInState=true}
  ]
  links [
   {
    dst=From Charge Transition
    rule=ItemFinalPoseOverride.AnimGraphNode_StateMachine_6.SprintFlight SM.AnimStateTransitionNode_30.AnimationTransitionGraph_0
    src=Conduit
    via=AnimStateTransitionNode_30
   }
   {
    dst=From Flight Transition
    rule=ItemFinalPoseOverride.AnimGraphNode_StateMachine_6.SprintFlight SM.AnimStateTransitionNode_29.AnimationTransitionGraph_0
    src=Conduit
    via=AnimStateTransitionNode_29
   }
   {
    dst=From Hover Transition
    rule=ItemFinalPoseOverride.AnimGraphNode_StateMachine_6.SprintFlight SM.AnimStateTransitionNode_26.AnimationTransitionGraph_0
    src=Conduit
    via=AnimStateTransitionNode_26
   }
   {
    dst=Safety Transition
    rule=ItemFinalPoseOverride.AnimGraphNode_StateMachine_6.SprintFlight SM.AnimStateTransitionNode_0.Transition
    src=Conduit
    via=AnimStateTransitionNode_0
   }
   {
    dst=Sprint Flight
    rule=ItemFinalPoseOverride.AnimGraphNode_StateMachine_6.SprintFlight SM.AnimStateTransitionNode_14.Transition
    src=Down Pitch React
    via=AnimStateTransitionNode_14
   }
   {
    dst=Sprint Flight
    rule=ItemFinalPoseOverride.AnimGraphNode_StateMachine_6.SprintFlight SM.AnimStateTransitionNode_4.Transition
    src=From Charge Transition
    via=AnimStateTransitionNode_4
   }
   {
    dst=Sprint Flight
    rule=ItemFinalPoseOverride.AnimGraphNode_StateMachine_6.SprintFlight SM.AnimStateTransitionNode_7.Transition
    src=From Flight Transition
    via=AnimStateTransitionNode_7
   }
   {
    dst=Sprint Flight
    rule=ItemFinalPoseOverride.AnimGraphNode_StateMachine_6.SprintFlight SM.AnimStateTransitionNode_2.Transition
    src=From Hover Transition
    via=AnimStateTransitionNode_2
   }
   {
    dst=Sprint Flight
    rule=ItemFinalPoseOverride.AnimGraphNode_StateMachine_6.SprintFlight SM.AnimStateTransitionNode_3.Transition
    src=Safety Transition
    via=AnimStateTransitionNode_3
   }
   {
    dst=Down Pitch React
    rule=ItemFinalPoseOverride.AnimGraphNode_StateMachine_6.SprintFlight SM.AnimStateTransitionNode_12.Transition
    src=Sprint Flight
    via=AnimStateTransitionNode_12
   }
   {
    dst=Up Pitch React
    rule=ItemFinalPoseOverride.AnimGraphNode_StateMachine_6.SprintFlight SM.AnimStateTransitionNode_11.Transition
    src=Sprint Flight
    via=AnimStateTransitionNode_11
   }
   {
    dst=Sprint Flight
    rule=ItemFinalPoseOverride.AnimGraphNode_StateMachine_6.SprintFlight SM.AnimStateTransitionNode_13.Transition
    src=Up Pitch React
    via=AnimStateTransitionNode_13
   }
  ]
 }
 Conduit {
  path=ItemFinalPoseOverride.AnimGraphNode_StateMachine_6.SprintFlight SM.AnimStateConduitNode_1.Conduit
  nodes [Result(bCanEnterTransition=true)]
 }
 Sprint Flight {
  path=ItemFinalPoseOverride.AnimGraphNode_StateMachine_6.SprintFlight SM.AnimStateNode_0.Sprint Flight
  nodes [Output Animation Pose(Result=Sequence Player 'IcicleLarva_Flight_OneHanded_Idle_Loop_Xtra'() {sequence=/IcicleLarvaGameplay/Assets/Animation/Flight/OneHandedFlight/IcicleLarva_Flight_OneHanded_Idle_Loop_Xtra}) {
    becomeRelevantFunction {bSelfContext=true, memberGuid=#29, memberName=FromIntroFalse}
    node {becomeRelevantFunction {functionName=FromIntroFalse}}
   }]
 }
 From Hover Transition {
  path=ItemFinalPoseOverride.AnimGraphNode_StateMachine_6.SprintFlight SM.AnimStateNode_1.From Hover Transition
  nodes [
   Modify Curve(CurveValues_0=1, Alpha=1) {curveNames [DisableFlightYawAdditive], curveValues [0]}
   Output Animation Pose(Result=Sequence Player 'Icicle_Larva_Hover_to_SprintFlying'() {bLoopAnimation=false, sequence=/IcicleLarvaGameplay/Assets/Animation/Flight/Icicle_Larva_Hover_to_SprintFlying}) {
    becomeRelevantFunction {bSelfContext=true, memberGuid=#15, memberName=FromHoverFalse}
    node {
     becomeRelevantFunction {functionName=FromHoverFalse}
     stateEntryFunction {functionName=OnStateEntry}
     stateExitFunction {functionName=OnStateExit}
    }
    stateEntryFunction {bSelfContext=true, memberGuid=#31, memberName=OnStateEntry_HoverToSprint}
    stateExitFunction {bSelfContext=true, memberGuid=#32, memberName=OnStateExit_HoverToSprint}
   }
  ]
 }
 Safety Transition {
  path=ItemFinalPoseOverride.AnimGraphNode_StateMachine_6.SprintFlight SM.AnimStateNode_2.Safety Transition
  nodes [
   Modify Curve(CurveValues_0=1, Alpha=1) {curveNames [DisableFlightYawAdditive], curveValues [0]}
   Output Animation Pose(Result=Sequence Player 'Icicle_Larva_Hover_to_SprintFlying'() {bLoopAnimation=false, sequence=&Icicle_Larva_Hover_to_SprintFlying}) {
    becomeRelevantFunction {bSelfContext=true, memberGuid=#15, memberName=FromHoverFalse}
    node {
     becomeRelevantFunction {functionName=FromHoverFalse}
     stateEntryFunction {functionName=OnStateEntry}
     stateExitFunction {functionName=OnStateExit}
    }
    stateEntryFunction {bSelfContext=true, memberGuid=#31, memberName=OnStateEntry_HoverToSprint}
    stateExitFunction {bSelfContext=true, memberGuid=#32, memberName=OnStateExit_HoverToSprint}
   }
  ]
 }
 From Flight Transition {
  path=ItemFinalPoseOverride.AnimGraphNode_StateMachine_6.SprintFlight SM.AnimStateNode_3.From Flight Transition
  nodes [Output Animation Pose(Result=Sequence Player 'IcicleLarva_Flight_Transition_FlighttoSprint'() {bLoopAnimation=false, sequence=/IcicleLarvaGameplay/Assets/Animation/Flight/IcicleLarva_Flight_Transition_FlighttoSprint}) {
    becomeRelevantFunction {bSelfContext=true, memberGuid=#16, memberName=FromFlightFalse}
    node {
     becomeRelevantFunction {functionName=FromFlightFalse}
     stateEntryFunction {functionName=OnStateEntry}
     stateExitFunction {functionName=OnStateExit}
    }
    stateEntryFunction {bSelfContext=true, memberGuid=#33, memberName=OnStateEntry_FlightToSprint}
    stateExitFunction {bSelfContext=true, memberGuid=#34, memberName=OnStateExit_FlightToSprint}
   }]
 }
 Up Pitch React {
  path=ItemFinalPoseOverride.AnimGraphNode_StateMachine_6.SprintFlight SM.AnimStateNode_4.Up Pitch React
  nodes [
   n1 = Sequence Player 'Icicle_Larva_Flight_OneHand_Trans_toUpFlying'!(StartPosition:float=0) -> (Pose) {bLoopAnimation=false, sequence=/IcicleLarvaGameplay/Assets/Animation/Flight/OneHandedFlight/Icicle_Larva_Flight_OneHand_Trans_toUpFlying}
   Output Animation Pose(Result=@n1)
  ]
 }
 Down Pitch React {
  path=ItemFinalPoseOverride.AnimGraphNode_StateMachine_6.SprintFlight SM.AnimStateNode_5.Down Pitch React
  nodes [
   n1 = Sequence Player 'Icicle_Larva_Flight_OneHand_Trans_toDownFlying'!(StartPosition:float=0) -> (Pose) {bLoopAnimation=false, sequence=/IcicleLarvaGameplay/Assets/Animation/Flight/OneHandedFlight/Icicle_Larva_Flight_OneHand_Trans_toDownFlying}
   Output Animation Pose(Result=@n1)
  ]
 }
 From Charge Transition {
  path=ItemFinalPoseOverride.AnimGraphNode_StateMachine_6.SprintFlight SM.AnimStateNode_8.From Charge Transition
  nodes [Output Animation Pose(Result=Sequence Player 'IcicleLarva_Flight_Transition_AirCharge_to_Sprint'() {bLoopAnimation=false, sequence=/IcicleLarvaGameplay/Assets/Animation/Flight/IcicleLarva_Flight_Transition_AirCharge_to_Sprint}) {
    becomeRelevantFunction {bSelfContext=true, memberGuid=#22, memberName=FromChargeFalse}
    node {becomeRelevantFunction {functionName=FromChargeFalse}}
   }]
 }
 Transition {
  path=ItemFinalPoseOverride.AnimGraphNode_StateMachine_6.SprintFlight SM.AnimStateTransitionNode_0.Transition
  nodes [Result(bCanEnterTransition=true)]
 }
 Transition {
  path=ItemFinalPoseOverride.AnimGraphNode_StateMachine_6.SprintFlight SM.AnimStateTransitionNode_11.Transition
  nodes [Result(bCanEnterTransition=Get bIsFlightUpPitchReaction())]
 }
 Transition {
  path=ItemFinalPoseOverride.AnimGraphNode_StateMachine_6.SprintFlight SM.AnimStateTransitionNode_12.Transition
  nodes [Result(bCanEnterTransition=Get bIsFlightDownPitchReaction())]
 }
 Transition {path=ItemFinalPoseOverride.AnimGraphNode_StateMachine_6.SprintFlight SM.AnimStateTransitionNode_13.Transition, nodes [Result()]}
 Transition {path=ItemFinalPoseOverride.AnimGraphNode_StateMachine_6.SprintFlight SM.AnimStateTransitionNode_14.Transition, nodes [Result()]}
 Transition {path=ItemFinalPoseOverride.AnimGraphNode_StateMachine_6.SprintFlight SM.AnimStateTransitionNode_2.Transition, nodes [Result()]}
 AnimationTransitionGraph_0 {
  path=ItemFinalPoseOverride.AnimGraphNode_StateMachine_6.SprintFlight SM.AnimStateTransitionNode_26.AnimationTransitionGraph_0
  nodes [Result(bCanEnterTransition=Get bIsFromHover())]
 }
 AnimationTransitionGraph_0 {
  path=ItemFinalPoseOverride.AnimGraphNode_StateMachine_6.SprintFlight SM.AnimStateTransitionNode_29.AnimationTransitionGraph_0
  nodes [Result(bCanEnterTransition=Get bIsFromFlight())]
 }
 Transition {path=ItemFinalPoseOverride.AnimGraphNode_StateMachine_6.SprintFlight SM.AnimStateTransitionNode_3.Transition, nodes [Result()]}
 AnimationTransitionGraph_0 {
  path=ItemFinalPoseOverride.AnimGraphNode_StateMachine_6.SprintFlight SM.AnimStateTransitionNode_30.AnimationTransitionGraph_0
  nodes [Result(bCanEnterTransition=Get bIsFromCharge())]
 }
 Transition {path=ItemFinalPoseOverride.AnimGraphNode_StateMachine_6.SprintFlight SM.AnimStateTransitionNode_4.Transition, nodes [Result()]}
 Transition {path=ItemFinalPoseOverride.AnimGraphNode_StateMachine_6.SprintFlight SM.AnimStateTransitionNode_7.Transition, nodes [Result()]}
 ItemFullBodyAimOffsets {nodes [
   Input Pose!() -> (Pose, FullBody_AimOffsetAlpha:double, Aim_Yaw:double, Aim_Pitch:double) {name=InPoseFullBodyAimOffsets}
   Input Pose!() -> (Pose) {name=InPosePreFullBodyAimOffsets}
   Output Pose() {layerGroup=Item Override}
  ]}
 ItemFullBodyOverride {nodes [
   AnimGraphNode_StateMachine_0 = FullBody PreAscend SM() {is=title AnimGraphNode_StateMachine}
   Output Pose(Result=@AnimGraphNode_StateMachine_0) {layerGroup=Item Override}
   Save cached pose 'FullBody Input Pose'(Pose=Input Pose() {name=InPoseFullBody}) {cacheName=FullBody Input Pose}
  ]}
 FullBody PreAscend SM {
  entry=Passthrough
  path=ItemFullBodyOverride.AnimGraphNode_StateMachine_0.FullBody PreAscend SM
  ids=id
  nodes [
   AnimStateNode_0 = Passthrough()
   AnimStateNode_1 = PreAscend Additive() {is=title AnimStateNode}
   AnimStateTransitionNode_0 = AnimStateTransitionNode()
   AnimStateTransitionNode_1 = AnimStateTransitionNode() {bAutomaticRuleBasedOnSequencePlayerInState=true, crossfadeDuration=0.4}
  ]
  links [
   {
    dst=PreAscend Additive
    rule=ItemFullBodyOverride.AnimGraphNode_StateMachine_0.FullBody PreAscend SM.AnimStateTransitionNode_0.Transition
    src=Passthrough
    via=AnimStateTransitionNode_0
   }
   {
    dst=Passthrough
    rule=ItemFullBodyOverride.AnimGraphNode_StateMachine_0.FullBody PreAscend SM.AnimStateTransitionNode_1.Transition
    src=PreAscend Additive
    via=AnimStateTransitionNode_1
   }
  ]
 }
 Passthrough {
  path=ItemFullBodyOverride.AnimGraphNode_StateMachine_0.FullBody PreAscend SM.AnimStateNode_0.Passthrough
  nodes [Output Animation Pose(Result=Use cached pose 'FullBody Input Pose'())]
 }
 PreAscend Additive {
  path=ItemFullBodyOverride.AnimGraphNode_StateMachine_0.FullBody PreAscend SM.AnimStateNode_1.PreAscend Additive
  nodes [
   n1 = "Sequence Player (additive) 'IcicleLarva_Ascend_Squat_Sprint_Additive'"!(PlayRate:float=1.1) -> (Pose) {
    bLoopAnimation=false
    groupName=Squat
    method=SyncGroup
    sequence=/IcicleLarvaGameplay/Assets/Animation/Proto/IcicleLarva_Ascend_Squat_Sprint_Additive
   }
   n2 = "Sequence Player (additive) 'IcicleLarva_Ascend_Squat_Additive'"(PlayRate=1.1) {
    bLoopAnimation=false
    groupName=Squat
    method=SyncGroup
    sequence=&IcicleLarva_Ascend_Squat_Additive
   }
   Blend Poses by bool(BlendPose_0=@n1, BlendPose_1=@n2, BlendTime_0=0.2, BlendTime_1=0.2, bActiveValue=Get bIsTacticalSprint() -> (bIsTacticalSprint:bool)) {blendType=Linear}
   Output Animation Pose(Result=Apply Additive(
     Base=Use cached pose 'FullBody Input Pose'()
     Additive="Sequence Player (additive) 'IcicleLarva_Ascend_Squat_Additive'"(PlayRate=1.1) {bLoopAnimation=false, sequence=&IcicleLarva_Ascend_Squat_Additive}
     Alpha=Get PreAscendAdditiveSquatValue() -> (PreAscendAdditiveSquatValue:float)
    ) {alphaBoolBlend {blendOption=Linear}})
  ]
 }
 Transition {
  path=ItemFullBodyOverride.AnimGraphNode_StateMachine_0.FullBody PreAscend SM.AnimStateTransitionNode_0.Transition
  nodes [Result(bCanEnterTransition=Get bIsPreAscend())]
 }
 Transition {path=ItemFullBodyOverride.AnimGraphNode_StateMachine_0.FullBody PreAscend SM.AnimStateTransitionNode_1.Transition, nodes [Result()]}
 ItemIdleOverride {nodes [
   Input Pose() {name=InPose_Idle}
   Output Pose() {layerGroup=Item Override}
  ]}
 ItemJetPackJumpAdditive {nodes [Output Pose() {layerGroup=Item Override}]}
 ItemJetPackStartAdditive {nodes [Output Pose() {layerGroup=Item Override}]}
 ItemJumpLoopAdditive {nodes [
   Input Pose() {name=InPoseItemJumpLoopAdditive}
   Output Pose() {layerGroup=Item Override}
  ]}
 ItemJumpOverride {nodes [
   Input Pose() {name=InPoseJump}
   Output Pose() {layerGroup=Item Override}
  ]}
 ItemJumpUpAdditive {nodes [
   Input Pose() {name=InPoseItemJumpUpAdditive}
   Output Pose() {layerGroup=Item Override}
  ]}
 ItemPreAimOffsets {nodes [
   Input Pose() {name=InPosePreAimOffsets}
   Output Pose() {layerGroup=Item Override}
  ]}
 ItemPreIK {nodes [
   Input Pose() {name=InPose_PreIK}
   Output Pose() {layerGroup=Item Override}
  ]}
 ItemRightHandIK {nodes [
   Input Pose() {name=InPosePreRightHandIK}
   Input Pose() {name=InPosePostRightHandIK}
   Output Pose() {layerGroup=Item Override}
  ]}
 ItemSkeletalControl {nodes [
   Input Pose()
   Output Pose() {layerGroup=Item Override}
  ]}
 ItemSwimJumpFallLoop {nodes [Output Pose() {layerGroup=Item Override}]}
 ItemSwimJumpSurfaceEnd {nodes [Output Pose() {layerGroup=Item Override}]}
 ItemSwimJumpSurfaceLoop {nodes [Output Pose() {layerGroup=Item Override}]}
 ItemUpperBody {nodes [
   AnimGraphNode_StateMachine_0 = HeatVision/Ascend UpperBody SM() {is=title AnimGraphNode_StateMachine}
   Output Pose(Result=@AnimGraphNode_StateMachine_0) {layerGroup=Item Override}
   Save cached pose 'UpperBodyInput Pose'(Pose=Input Pose!() -> (Pose, InputParam:Fort Anim Input Adjusted Aim Structure) {name=InPoseUpperBody}) {cacheName=UpperBodyInput Pose}
  ]}
 HeatVision/Ascend UpperBody SM {
  entry=Pass Through
  path=ItemUpperBody.AnimGraphNode_StateMachine_0.HeatVision/Ascend UpperBody SM
  ids=id
  nodes [
   To Pre-Ascend() {is=title AnimStateAliasNode, also {aliases [HeatVision Intro, HeatVision Loop, HeatVision Outro, Pass Through]}}
   AnimStateConduitNode_0 = HeatVision() {is=title AnimStateConduitNode}
   AnimStateNode_0 = Pre-Ascend() {is=title AnimStateNode}
   AnimStateNode_4 = HeatVision Intro()
   AnimStateNode_5 = HeatVision Loop()
   AnimStateNode_6 = HeatVision Outro()
   AnimStateNode_7 = Pass Through()
   AnimStateTransitionNode_0 = AnimStateTransitionNode() {bAutomaticRuleBasedOnSequencePlayerInState=true, crossfadeDuration=0.4}
   AnimStateTransitionNode_10 = AnimStateTransitionNode()
   AnimStateTransitionNode_11 = AnimStateTransitionNode() {bAutomaticRuleBasedOnSequencePlayerInState=true, priorityOrder=10}
   AnimStateTransitionNode_12 = AnimStateTransitionNode() {priorityOrder=10}
   AnimStateTransitionNode_15 = AnimStateTransitionNode() {bAutomaticRuleBasedOnSequencePlayerInState=true}
   AnimStateTransitionNode_18 = AnimStateTransitionNode()
   AnimStateTransitionNode_2 = AnimStateTransitionNode() {bDisabled=true, crossfadeDuration=0.25}
   AnimStateTransitionNode_3 = AnimStateTransitionNode() {priorityOrder=20}
   AnimStateTransitionNode_4 = AnimStateTransitionNode()
   AnimStateTransitionNode_7 = AnimStateTransitionNode()
   AnimStateTransitionNode_9 = AnimStateTransitionNode()
  ]
  links [
   {
    dst=HeatVision Intro
    rule=ItemUpperBody.AnimGraphNode_StateMachine_0.HeatVision/Ascend UpperBody SM.AnimStateTransitionNode_3.Transition
    src=HeatVision
    via=AnimStateTransitionNode_3
   }
   {
    dst=HeatVision Intro
    rule=ItemUpperBody.AnimGraphNode_StateMachine_0.HeatVision/Ascend UpperBody SM.AnimStateTransitionNode_9.Transition
    src=HeatVision
    via=AnimStateTransitionNode_9
   }
   {
    dst=HeatVision Loop
    rule=ItemUpperBody.AnimGraphNode_StateMachine_0.HeatVision/Ascend UpperBody SM.AnimStateTransitionNode_4.Transition
    src=HeatVision
    via=AnimStateTransitionNode_4
   }
   {
    dst=HeatVision Outro
    rule=ItemUpperBody.AnimGraphNode_StateMachine_0.HeatVision/Ascend UpperBody SM.AnimStateTransitionNode_7.Transition
    src=HeatVision
    via=AnimStateTransitionNode_7
   }
   {
    dst=HeatVision Loop
    rule=ItemUpperBody.AnimGraphNode_StateMachine_0.HeatVision/Ascend UpperBody SM.AnimStateTransitionNode_11.Transition
    src=HeatVision Intro
    via=AnimStateTransitionNode_11
   }
   {
    dst=HeatVision Outro
    rule=ItemUpperBody.AnimGraphNode_StateMachine_0.HeatVision/Ascend UpperBody SM.AnimStateTransitionNode_12.Transition
    src=HeatVision Loop
    via=AnimStateTransitionNode_12
   }
   {
    dst=Pass Through
    rule=ItemUpperBody.AnimGraphNode_StateMachine_0.HeatVision/Ascend UpperBody SM.AnimStateTransitionNode_15.Transition
    src=HeatVision Outro
    via=AnimStateTransitionNode_15
   }
   {
    dst=HeatVision
    rule=ItemUpperBody.AnimGraphNode_StateMachine_0.HeatVision/Ascend UpperBody SM.AnimStateTransitionNode_10.Transition
    src=Pass Through
    via=AnimStateTransitionNode_10
   }
   {
    dst=HeatVision Outro
    rule=ItemUpperBody.AnimGraphNode_StateMachine_0.HeatVision/Ascend UpperBody SM.AnimStateTransitionNode_18.Transition
    src=Pass Through
    via=AnimStateTransitionNode_18
   }
   {
    dst=Pass Through
    rule=ItemUpperBody.AnimGraphNode_StateMachine_0.HeatVision/Ascend UpperBody SM.AnimStateTransitionNode_0.Transition
    src=Pre-Ascend
    via=AnimStateTransitionNode_0
   }
   {
    dst=Pre-Ascend
    rule=ItemUpperBody.AnimGraphNode_StateMachine_0.HeatVision/Ascend UpperBody SM.AnimStateTransitionNode_2.Transition
    src=To Pre-Ascend
    via=AnimStateTransitionNode_2
   }
  ]
 }
 HeatVision {
  path=ItemUpperBody.AnimGraphNode_StateMachine_0.HeatVision/Ascend UpperBody SM.AnimStateConduitNode_0.HeatVision
  nodes [Result(bCanEnterTransition=true)]
 }
 Pre-Ascend {
  path=ItemUpperBody.AnimGraphNode_StateMachine_0.HeatVision/Ascend UpperBody SM.AnimStateNode_0.Pre-Ascend
  nodes [Output Animation Pose(Result=Sequence Player 'IcicleLarva_Ascend_New_Trimmed'!(PlayRate:float=1.2, StartPosition:float=0.47) -> (Pose) {bLoopAnimation=false, sequence=/IcicleLarvaGameplay/Assets/Animation/Proto/IcicleLarva_Ascend_New_Trimmed})]
 }
 HeatVision Intro {
  path=ItemUpperBody.AnimGraphNode_StateMachine_0.HeatVision/Ascend UpperBody SM.AnimStateNode_4.HeatVision Intro
  nodes [
   n1 = Sequence Player 'IcicleLarva_HeatVision_Firing_Intro'() {
    bLoopAnimation=false
    groupName=HeatVision
    method=SyncGroup
    sequence=&IcicleLarva_HeatVision_Firing_Intro
   }
   n2 = Layered blend per bone(BasePose=Use cached pose 'UpperBodyInput Pose'(), BlendPoses_0=@n1) {bMeshSpaceRotationBlend=true, layerSetup [{branchFilters [{blendDepth=1, boneName=neck_01}, {boneName=clavicle_l}, {boneName=clavicle_r}]}]}
   n3 = Sequence Player 'IcicleLarva_HeatVision_Firing_Intro'() {
    bLoopAnimation=false
    groupName=HeatVision
    method=SyncGroup
    sequence=&IcicleLarva_HeatVision_Firing_Intro
   }
   n4 = Get bIsCrouchMoving() -> (bIsCrouchMoving:bool)
   n5 = Sequence Player 'IcicleLarva_HeatVision_InAir_Firing_Intro'() {
    bLoopAnimation=false
    groupName=HeatVision
    method=SyncGroup
    sequence=&IcicleLarva_HeatVision_InAir_Firing_Intro
   }
   Output Animation Pose(Result=Blend Poses by bool(
     BlendPose_0=@n5
     BlendPose_1=Blend Poses by bool(BlendPose_0=@n2, BlendPose_1=@n3, BlendTime_0=0.2, BlendTime_1=0.2, bActiveValue=@n4) {blendType=Linear}
     BlendTime_0=0.2
     BlendTime_1=0.2
     bActiveValue=Get bIsHover()
    ) {blendType=Linear})
  ]
 }
 HeatVision Loop {
  path=ItemUpperBody.AnimGraphNode_StateMachine_0.HeatVision/Ascend UpperBody SM.AnimStateNode_5.HeatVision Loop
  nodes [
   n1 = Layered blend per bone(
    BasePose=Use cached pose 'UpperBodyInput Pose'()
    BlendPoses_0=Sequence Player 'IcicleLarva_HeatVision_Firing_Loop'() {bLoopAnimation=false, sequence=&IcicleLarva_HeatVision_Firing_Loop}
   ) {bMeshSpaceRotationBlend=true, layerSetup [{branchFilters [{blendDepth=1, boneName=neck_01}, {boneName=clavicle_l}, {boneName=clavicle_r}]}]}
   n2 = Sequence Player 'IcicleLarva_HeatVision_Firing_Loop'() {bLoopAnimation=false, sequence=&IcicleLarva_HeatVision_Firing_Loop}
   n3 = Get bIsCrouchMoving()
   Output Animation Pose(Result=Blend Poses by bool(
     BlendPose_0=Sequence Player 'IcicleLarva_HeatVision_InAir_Firing_Loop'() {bLoopAnimation=false, sequence=&IcicleLarva_HeatVision_InAir_Firing_Loop}
     BlendPose_1=Blend Poses by bool(BlendPose_0=@n1, BlendPose_1=@n2, BlendTime_0=0.2, BlendTime_1=0.2, bActiveValue=@n3) {blendType=Linear}
     BlendTime_0=0.2
     BlendTime_1=0.2
     bActiveValue=Get bIsHover()
    ) {blendType=Linear})
  ]
 }
 HeatVision Outro {
  path=ItemUpperBody.AnimGraphNode_StateMachine_0.HeatVision/Ascend UpperBody SM.AnimStateNode_6.HeatVision Outro
  nodes [
   n1 = Sequence Player 'IcicleLarva_HeatVision_Firing_Outro'() {
    bLoopAnimation=false
    groupName=HeatVisionOutro
    method=SyncGroup
    sequence=&IcicleLarva_HeatVision_Firing_Outro
   }
   n2 = Layered blend per bone(BasePose=Use cached pose 'UpperBodyInput Pose'(), BlendPoses_0=@n1) {bMeshSpaceRotationBlend=true, layerSetup [{branchFilters [{blendDepth=1, boneName=neck_01}, {boneName=clavicle_l}, {boneName=clavicle_r}]}]}
   n3 = Sequence Player 'IcicleLarva_HeatVision_Firing_Outro'() {
    bLoopAnimation=false
    groupName=HeatVisionOutro
    method=SyncGroup
    sequence=&IcicleLarva_HeatVision_Firing_Outro
   }
   n4 = Get bIsCrouchMoving()
   n5 = Sequence Player 'IcicleLarva_HeatVision_InAir_Firing_Outro'() {
    bLoopAnimation=false
    groupName=HeatVisionOutro
    method=SyncGroup
    sequence=&IcicleLarva_HeatVision_InAir_Firing_Outro
   }
   Output Animation Pose(Result=Blend Poses by bool(
     BlendPose_0=@n5
     BlendPose_1=Blend Poses by bool(BlendPose_0=@n2, BlendPose_1=@n3, BlendTime_0=0.2, BlendTime_1=0.2, bActiveValue=@n4) {blendType=Linear}
     BlendTime_0=0.2
     BlendTime_1=0.2
     bActiveValue=Get bIsHover()
    ) {blendType=Linear})
  ]
 }
 Pass Through {
  path=ItemUpperBody.AnimGraphNode_StateMachine_0.HeatVision/Ascend UpperBody SM.AnimStateNode_7.Pass Through
  nodes [Output Animation Pose(Result=Use cached pose 'UpperBodyInput Pose'())]
 }
 Transition {path=ItemUpperBody.AnimGraphNode_StateMachine_0.HeatVision/Ascend UpperBody SM.AnimStateTransitionNode_0.Transition, nodes [Result()]}
 Transition {
  path=ItemUpperBody.AnimGraphNode_StateMachine_0.HeatVision/Ascend UpperBody SM.AnimStateTransitionNode_10.Transition
  nodes [Result(bCanEnterTransition=Get bIsHeatVision())]
 }
 Transition {path=ItemUpperBody.AnimGraphNode_StateMachine_0.HeatVision/Ascend UpperBody SM.AnimStateTransitionNode_11.Transition, nodes [Result()]}
 Transition {
  path=ItemUpperBody.AnimGraphNode_StateMachine_0.HeatVision/Ascend UpperBody SM.AnimStateTransitionNode_12.Transition
  nodes [Result(bCanEnterTransition=NOT Boolean(A=Get bIsHeatVision()))]
 }
 Transition {path=ItemUpperBody.AnimGraphNode_StateMachine_0.HeatVision/Ascend UpperBody SM.AnimStateTransitionNode_15.Transition, nodes [Result()]}
 Transition {
  path=ItemUpperBody.AnimGraphNode_StateMachine_0.HeatVision/Ascend UpperBody SM.AnimStateTransitionNode_18.Transition
  nodes [Result(bCanEnterTransition=Get bIsHeatVisionOutro() -> (bIsHeatVisionOutro:bool))]
 }
 Transition {
  path=ItemUpperBody.AnimGraphNode_StateMachine_0.HeatVision/Ascend UpperBody SM.AnimStateTransitionNode_2.Transition
  nodes [Result(bCanEnterTransition=Get bIsPreAscend())]
 }
 Transition {
  path=ItemUpperBody.AnimGraphNode_StateMachine_0.HeatVision/Ascend UpperBody SM.AnimStateTransitionNode_3.Transition
  nodes [Result(bCanEnterTransition=Get bIsHeatVision())]
 }
 Transition {
  path=ItemUpperBody.AnimGraphNode_StateMachine_0.HeatVision/Ascend UpperBody SM.AnimStateTransitionNode_4.Transition
  nodes [Result(bCanEnterTransition=Get bIsHeatVisionLoop() -> (bIsHeatVisionLoop:bool))]
 }
 Transition {
  path=ItemUpperBody.AnimGraphNode_StateMachine_0.HeatVision/Ascend UpperBody SM.AnimStateTransitionNode_7.Transition
  nodes [Result(bCanEnterTransition=Get bIsHeatVisionOutro())]
 }
 Transition {
  path=ItemUpperBody.AnimGraphNode_StateMachine_0.HeatVision/Ascend UpperBody SM.AnimStateTransitionNode_9.Transition
  nodes [Result(bCanEnterTransition=Get bIsHeatVisionIntro() -> (bIsHeatVisionIntro:bool))]
 }
 ItemZipLineStartAdditive {nodes [Output Pose() {layerGroup=Item Override}]}
 Normal Flight Adjustments {nodes [
   AnimGraphNode_StateMachine_0 = Normal Flight Adjustment SM() {is=title AnimGraphNode_StateMachine}
   Output Pose(Result=@AnimGraphNode_StateMachine_0) {layerGroup=Flight Adjustments}
   AnimGraphNode_StateMachine_3 = Settle() {is=title AnimGraphNode_StateMachine}
   Save cached pose 'Settle Additive'(Pose=@AnimGraphNode_StateMachine_3) {cacheName=Settle Additive}
   Save cached pose 'Normal Flight Adjustment Input Pose'(Pose=Input Pose!() -> (Pose) {name=FlightInPose}) {cacheName=Normal Flight Adjustment Input Pose}
  ]}
 Normal Flight Adjustment SM {
  entry=Normal Flight Adjustment
  path=Normal Flight Adjustments.AnimGraphNode_StateMachine_0.Normal Flight Adjustment SM
  ids=id
  nodes [AnimStateNode_0 = Normal Flight Adjustment() {is=title AnimStateNode}]
 }
 Normal Flight Adjustment {
  path=Normal Flight Adjustments.AnimGraphNode_StateMachine_0.Normal Flight Adjustment SM.AnimStateNode_0.Normal Flight Adjustment
  nodes [
   n1 = Layered blend per bone(
    BasePose=Additive Identity Pose()
    BlendPoses_0=Blendspace Player 'IcicleLarva_Flight_HandsBack_BS'(X=0, Y=0) {blendSpace=/IcicleLarvaGameplay/Assets/Animation/Flight/HandsBackFlight/IcicleLarva_Flight_HandsBack_BS}
   ) {layerSetup [{branchFilters [{boneName=Pelvis}, {blendDepth=-1, boneName=spine_01}, {blendDepth=-1, boneName=thigh_l}, {blendDepth=-1, boneName=thigh_r}]}]}
   n2 = Apply Additive(
    Base=Apply Additive(
     Base=Use cached pose 'Normal Flight Adjustment Input Pose'()
     Additive="Sequence Player (additive) 'IcicleLarva_Flight_Sway_Idle_HandsBack_Loop'"() {sequence=/IcicleLarvaGameplay/Assets/Animation/Flight/HandsBackFlight/IcicleLarva_Flight_Sway_Idle_HandsBack_Loop}
    ) {alphaBoolBlend {blendOption=Linear}}
    Additive=@n1
   ) {alphaBoolBlend {blendOption=Linear}}
   n3 = Layered blend per bone(
    BasePose=Additive Identity Pose()
    BlendPoses_0=Blend Poses by bool(
     BlendPose_0=Blendspace Player 'IcicleLarva_Flight_HandsBack_BS'(X=0, Y=0) {blendSpace=&IcicleLarva_Flight_HandsBack_BS}
     BlendPose_1=Blendspace Player 'IcicleLarva_Flight_HandsBack_BS'(X=0, Y=0) {blendSpace=&IcicleLarva_Flight_HandsBack_BS}
    ) {blendType=Linear}
   ) {layerSetup [{branchFilters [{boneName=spine_01}, {boneName=thigh_l}, {boneName=thigh_r}]}]}
   n4 = Layered blend per bone(
    BasePose=Additive Identity Pose()
    BlendPoses_0=Blendspace Player 'IcicleLarva_Flight_HandsBack_BS'(X=0, Y=0) {blendSpace=&IcicleLarva_Flight_HandsBack_BS}
   ) {layerSetup [{branchFilters [{boneName=spine_01}, {boneName=thigh_l}, {boneName=thigh_r}]}]}
   n5 = Apply Additive(
    Base=Apply Additive(Base=Apply Additive(Base=@n2, Additive=@n3, Alpha=0.9) {alphaBoolBlend {blendOption=Linear}}, Additive=@n4, Alpha=0.75) {alphaBoolBlend {blendOption=Linear}}
    Additive=Blendspace Player 'IcicleLarva_Flight_HandsBack_1D_BS'!(X=0) -> (Pose) {blendSpace=/IcicleLarvaGameplay/Assets/Animation/Flight/HandsBackFlight/IcicleLarva_Flight_HandsBack_1D_BS}
    Alpha=0.5
   ) {alphaBoolBlend {blendOption=Linear}}
   n6 = Use cached pose 'Settle Additive'()
   Output Animation Pose(Result=Apply Additive(Base=Apply Additive(Base=@n5, Additive=@n6) {alphaBoolBlend {blendOption=Linear}}, Additive=Use cached pose 'Settle Additive'()) {alphaBoolBlend {blendOption=Linear}})
  ]
 }
 Settle {
  entry=Passthrough
  path=Normal Flight Adjustments.AnimGraphNode_StateMachine_3.Settle
  ids=id
  nodes [
   Exit Settle() {also {aliases [Start Settle]}}
   AnimStateNode_0 = Passthrough()
   AnimStateNode_1 = Start Settle()
   AnimStateNode_2 = Settle() {is=title AnimStateNode}
   AnimStateTransitionNode_0 = AnimStateTransitionNode() {crossfadeDuration=0}
   AnimStateTransitionNode_2 = AnimStateTransitionNode() {crossfadeDuration=1}
   AnimStateTransitionNode_3 = AnimStateTransitionNode() {bAutomaticRuleBasedOnSequencePlayerInState=true}
   AnimStateTransitionNode_5 = AnimStateTransitionNode()
   AnimStateTransitionNode_6 = AnimStateTransitionNode()
  ]
  links [
   {
    dst=Passthrough
    rule=Normal Flight Adjustments.AnimGraphNode_StateMachine_3.Settle.AnimStateTransitionNode_5.Transition
    src=Exit Settle
    via=AnimStateTransitionNode_5
   }
   {
    dst=Passthrough
    rule=Normal Flight Adjustments.AnimGraphNode_StateMachine_3.Settle.AnimStateTransitionNode_6.Transition
    src=Exit Settle
    via=AnimStateTransitionNode_6
   }
   {
    dst=Start Settle
    rule=Normal Flight Adjustments.AnimGraphNode_StateMachine_3.Settle.AnimStateTransitionNode_0.Transition
    src=Passthrough
    via=AnimStateTransitionNode_0
   }
   {
    dst=Passthrough
    rule=Normal Flight Adjustments.AnimGraphNode_StateMachine_3.Settle.AnimStateTransitionNode_3.Transition
    src=Settle
    via=AnimStateTransitionNode_3
   }
   {
    dst=Settle
    rule=Normal Flight Adjustments.AnimGraphNode_StateMachine_3.Settle.AnimStateTransitionNode_2.Transition
    src=Start Settle
    via=AnimStateTransitionNode_2
   }
  ]
 }
 Passthrough {
  path=Normal Flight Adjustments.AnimGraphNode_StateMachine_3.Settle.AnimStateNode_0.Passthrough
  nodes [Output Animation Pose(Result=Additive Identity Pose())]
 }
 Start Settle {
  path=Normal Flight Adjustments.AnimGraphNode_StateMachine_3.Settle.AnimStateNode_1.Start Settle
  nodes [Output Animation Pose(Result=Additive Identity Pose()) {
    node {stateEntryFunction {functionName=OnStateEntry}}
    stateEntryFunction {bSelfContext=true, memberGuid=#1, memberName=OnStateEntry}
   }]
 }
 Settle {
  path=Normal Flight Adjustments.AnimGraphNode_StateMachine_3.Settle.AnimStateNode_2.Settle
  nodes [
   n1 = "Sequence Player (additive) 'IcicleLarva_Flight_Additive_BackLeft_Legs_Settle'"() {bLoopAnimation=false, sequence=&IcicleLarva_Flight_Additive_BackLeft_Legs_Settle}
   n2 = "Sequence Player (additive) 'IcicleLarva_Flight_Additive_BankRight_Legs_Settle'"() {
    bLoopAnimation=false
    sequence=&IcicleLarva_Flight_Additive_BankRight_Legs_Settle
   }
   Output Animation Pose(Result=Blend Poses by bool(BlendPose_0=@n1, BlendPose_1=@n2, bActiveValue=Get bSettleDirectionBlend()) {blendType=Linear})
  ]
 }
 Transition {
  path=Normal Flight Adjustments.AnimGraphNode_StateMachine_3.Settle.AnimStateTransitionNode_0.Transition
  nodes [Result(bCanEnterTransition=Get bPassThroughToStartSettleTransition())]
 }
 Transition {
  path=Normal Flight Adjustments.AnimGraphNode_StateMachine_3.Settle.AnimStateTransitionNode_2.Transition
  nodes [Result(bCanEnterTransition=true)]
 }
 Transition {path=Normal Flight Adjustments.AnimGraphNode_StateMachine_3.Settle.AnimStateTransitionNode_3.Transition, nodes [Result()]}
 Transition {
  path=Normal Flight Adjustments.AnimGraphNode_StateMachine_3.Settle.AnimStateTransitionNode_5.Transition
  nodes [Result(bCanEnterTransition=Get bExitSettleToPassThroughSecondTransition())]
 }
 Transition {
  path=Normal Flight Adjustments.AnimGraphNode_StateMachine_3.Settle.AnimStateTransitionNode_6.Transition
  nodes [Result(bCanEnterTransition=Get bExitSettleToPassThroughTransition())]
 }
 Normal Flight Blend Adjustments {nodes [
   AnimGraphNode_StateMachine_0 = Normal Flight Blend Adjustment SM() {is=title AnimGraphNode_StateMachine}
   Output Pose(Result=Modify Curve(SourcePose=@AnimGraphNode_StateMachine_0, CurveValues_0=0, Alpha=1) {curveNames [HoverPivotAlpha], curveValues [0]}) {layerGroup=Flight Adjustments}
   Save cached pose 'Normal Flight Blend Input Pose'(Pose=Input Pose() {name=InPose_0}) {cacheName=Normal Flight Blend Input Pose}
  ]}
 Normal Flight Blend Adjustment SM {
  entry=Normal Flight Blend
  path=Normal Flight Blend Adjustments.AnimGraphNode_StateMachine_0.Normal Flight Blend Adjustment SM
  ids=id
  nodes [AnimStateNode_0 = Normal Flight Blend() {is=title AnimStateNode}]
 }
 Normal Flight Blend {
  path=Normal Flight Blend Adjustments.AnimGraphNode_StateMachine_0.Normal Flight Blend Adjustment SM.AnimStateNode_0.Normal Flight Blend
  nodes [
   n1 = Layered blend per bone(
    BasePose=Additive Identity Pose()
    BlendPoses_0=Blendspace Player 'IcicleLarva_Flight_HandsBack_BS'(X=0, Y=0) {blendSpace=&IcicleLarva_Flight_HandsBack_BS}
   ) {layerSetup [{branchFilters [{boneName=Pelvis}, {blendDepth=-1, boneName=spine_01}, {blendDepth=-1, boneName=thigh_l}, {blendDepth=-1, boneName=thigh_r}]}]}
   n2 = Apply Additive(
    Base=Apply Additive(
     Base=Use cached pose 'Normal Flight Blend Input Pose'()
     Additive="Sequence Player (additive) 'IcicleLarva_Flight_Sway_Idle_HandsBack_Loop'"() {sequence=&IcicleLarva_Flight_Sway_Idle_HandsBack_Loop}
    ) {alphaBoolBlend {blendOption=Linear}}
    Additive=@n1
   ) {alphaBoolBlend {blendOption=Linear}}
   n3 = Layered blend per bone(
    BasePose=Additive Identity Pose()
    BlendPoses_0=Blendspace Player 'IcicleLarva_Flight_HandsBack_BS'(X=0, Y=0) {blendSpace=&IcicleLarva_Flight_HandsBack_BS}
   ) {layerSetup [{branchFilters [{boneName=spine_01}, {boneName=thigh_l}, {boneName=thigh_r}]}]}
   n4 = Layered blend per bone(
    BasePose=Additive Identity Pose()
    BlendPoses_0=Blendspace Player 'IcicleLarva_Flight_HandsBack_BS'(X=0, Y=0) {blendSpace=&IcicleLarva_Flight_HandsBack_BS}
   ) {layerSetup [{branchFilters [{boneName=spine_01}, {boneName=thigh_l}, {boneName=thigh_r}]}]}
   Output Animation Pose(Result=Apply Additive(Base=Apply Additive(Base=@n2, Additive=@n3, Alpha=0.9) {alphaBoolBlend {blendOption=Linear}}, Additive=@n4, Alpha=0.75) {alphaBoolBlend {blendOption=Linear}})
  ]
 }
 OnBecomeRelevant_HoverPivot {nodes [
   n1 = OnBecomeRelevant_HoverPivot!() -> (then, Context:"Anim Update Context Structure (by ref)", Node:"Anim Node Reference Structure (by ref)") {is=title K2Node_FunctionEntry}
   n2 = "float > float"(A=Get ForwardBackwardPivotDirection() -> (ForwardBackwardPivotDirection:double))
   n3 = Get HoverPivotLeftToRight() -> (HoverPivotLeftToRight:Anim Sequence Base Object Reference)
   n4 = Get HoverPivotRightToLeft() -> (HoverPivotRightToLeft:Anim Sequence Base Object Reference)
   n5 = "float > float"(A=Get LeftRightPivotDirection() -> (LeftRightPivotDirection:double))
   n6 = Get HoverPivotForwardToBack() -> (HoverPivotForwardToBack:Anim Sequence Base Object Reference)
   n7 = Get HoverPivotBackToForward() -> (HoverPivotBackToForward:Anim Sequence Base Object Reference)
   > SetSequenceWithInertialBlending!(
    execute
    UpdateContext:"Anim Update Context Structure (by ref)"=@n1
    SequencePlayer:"Sequence Player Reference Structure (by ref)"=Convert to Sequence Player!(Node:"Anim Node Reference Structure (by ref)"=@n1.Node) -> (SequencePlayer:Sequence Player Reference Structure, Result:bool)
    Sequence:Anim Sequence Base Object Reference=Select(
     Option 0:Anim Sequence Base Object Reference=Select(Option 0=@n4, Option 1=@n3, Index=@n2)
     Option 1:Anim Sequence Base Object Reference=Select(Option 0=@n7, Option 1=@n6, Index=@n5)
     Index=Get bIsForwardBackPivot() -> (bIsForwardBackPivot:bool)
    ) -> (ReturnValue:Anim Sequence Base Object Reference)
    BlendTime:float=0.2
   ) -> (then, ReturnValue:Sequence Player Reference Structure)
   > Set bIsInitialForwardBackDirection(bIsInitialForwardBackDirection:bool=Get bIsForwardBackPivot()) -> (Output_Get:bool)
   > Set InitialForwardDirection(InitialForwardDirection:double=Get ForwardBackwardPivotDirection()) -> (Output_Get:double)
   > Return Node()
  ]}
 OnStateEntry_FlightToSprint {nodes [
   OnStateEntry_FlightToSprint!() -> (then, Context:"Anim Update Context Structure (by ref)", Node:"Anim Node Reference Structure (by ref)") {is=title K2Node_FunctionEntry}
   > Set bFlightToSprint(bFlightToSprint:bool=true) -> (Output_Get:bool)
   > Return Node()
  ]}
 OnStateEntry_HoverToSprint {nodes [
   OnStateEntry_HoverToSprint!() -> (then, Context:"Anim Update Context Structure (by ref)", Node:"Anim Node Reference Structure (by ref)") {is=title K2Node_FunctionEntry}
   > Set bHoverToSprint(bHoverToSprint:bool=true) -> (Output_Get:bool)
   > Return Node()
  ]}
 OnStateEntry_Settle {nodes [
   OnStateEntry_Settle!() -> (then, Context:"Anim Update Context Structure (by ref)", Node:"Anim Node Reference Structure (by ref)") {is=title K2Node_FunctionEntry}
   > Set InitialSettleDirection(InitialSettleDirection:double=Get FlightYawDelta() -> (FlightYawDelta:double)) -> (Output_Get:double)
   > Return Node()
  ]}
 OnStateExit_FlightToSprint {nodes [
   OnStateExit_FlightToSprint!() -> (then, Context:"Anim Update Context Structure (by ref)", Node:"Anim Node Reference Structure (by ref)") {is=title K2Node_FunctionEntry}
   > Set bFlightToSprint(bFlightToSprint=false)
   > Return Node()
  ]}
 OnStateExit_HoverToSprint {nodes [
   OnStateExit_HoverToSprint!() -> (then, Context:"Anim Update Context Structure (by ref)", Node:"Anim Node Reference Structure (by ref)") {is=title K2Node_FunctionEntry}
   > Set bHoverToSprint(bHoverToSprint=false)
   > Return Node()
  ]}
 ResetFlightYawReactReaction {nodes [
   ResetFlightYawReactReaction!() -> (then, Context:"Anim Update Context Structure (by ref)", Node:"Anim Node Reference Structure (by ref)") {is=title K2Node_FunctionEntry}
   > Set bIsFlightYawReaction(bIsFlightYawReaction:bool=false) -> (Output_Get:bool)
   > Return Node()
  ]}
 Sprint Flight Adjustments {nodes [
   AnimGraphNode_StateMachine_0 = Sprint Flight Adjustment SM() {is=title AnimGraphNode_StateMachine}
   Output Pose(Result=@AnimGraphNode_StateMachine_0) {layerGroup=Flight Adjustments}
   Save cached pose 'Sprint Flight Adjustment Input Pose'(Pose=Input Pose() {name=SprintInPose}) {cacheName=Sprint Flight Adjustment Input Pose}
  ]}
 Sprint Flight Adjustment SM {
  entry=Sprint Adjustment
  path=Sprint Flight Adjustments.AnimGraphNode_StateMachine_0.Sprint Flight Adjustment SM
  ids=id
  nodes [AnimStateNode_0 = Sprint Adjustment() {is=title AnimStateNode, bAlwaysResetOnEntry=true}]
 }
 Sprint Adjustment {
  path=Sprint Flight Adjustments.AnimGraphNode_StateMachine_0.Sprint Flight Adjustment SM.AnimStateNode_0.Sprint Adjustment
  nodes [
   n1 = "Sequence Player (additive) 'IcicleLarva_Flight_Sway_Idle_OneHanded'"() {bLoopAnimation=false, sequence=/IcicleLarvaGameplay/Assets/Animation/Flight/OneHandedFlight/IcicleLarva_Flight_Sway_Idle_OneHanded}
   n2 = Use cached pose 'Sprint Flight Adjustment Input Pose'()
   n3 = Layered blend per bone(
    BasePose=Additive Identity Pose()
    BlendPoses_0=Two Way Blend(
     A=Blendspace Player 'IcicleLarva_Flight_Pelvis_Sprint_BS'(X=0, Y=0) {blendSpace=/IcicleLarvaGameplay/Assets/Animation/Flight/OneHandedFlight/IcicleLarva_Flight_Pelvis_Sprint_BS}
     B=Blendspace Player 'IcicleLarva_Flight_TwoHanded_BS_Old'(X=0, Y=0) {blendSpace=&IcicleLarva_Flight_TwoHanded_BS_Old}
     Alpha=Get YawAlpha() -> (YawAlpha:double)
    ) {blendNode {alphaBoolBlend {blendOption=Linear}}}
   ) {layerSetup [{branchFilters [{boneName=Pelvis}, {blendDepth=-1, boneName=spine_01}, {blendDepth=-1, boneName=thigh_l}, {blendDepth=-1, boneName=thigh_r}]}]}
   n4 = Layered blend per bone(
    BasePose=Additive Identity Pose()
    BlendPoses_0=Blendspace Player 'IcicleLarva_Flight_Sprint_BS'(X=0, Y=0) {blendSpace=/IcicleLarvaGameplay/Assets/Animation/Flight/OneHandedFlight/IcicleLarva_Flight_Sprint_BS}
   ) {layerSetup [{branchFilters [{boneName=spine_01}, {boneName=thigh_l}, {boneName=thigh_r}, {blendDepth=-1, boneName=hand_r}, {blendDepth=-1, boneName=hand_l}]}]}
   n5 = Apply Additive(
    Base=Apply Additive(Base=Apply Additive(Base=@n2, Additive=@n1) {alphaBoolBlend {blendOption=Linear}}, Additive=@n3) {alphaBoolBlend {blendOption=Linear}}
    Additive=@n4
    Alpha=0.5
   ) {alphaBoolBlend {blendOption=Linear}}
   n6 = Layered blend per bone(
    BasePose=Additive Identity Pose()
    BlendPoses_0=Blendspace Player 'IcicleLarva_Flight_Sprint_BS'(X=0, Y=0) {blendSpace=&IcicleLarva_Flight_Sprint_BS}
   ) {layerSetup [{branchFilters [{boneName=spine_01}, {boneName=thigh_l}, {boneName=thigh_r}]}]}
   Output Animation Pose(Result=Apply Additive(
     Base=Apply Additive(Base=@n5, Additive=@n6) {alphaBoolBlend {blendOption=Linear}}
     Additive="Sequence Player (additive) 'IcicleLarva_Flight_Noise_Idle_OneHanded_Loop'"() {sequence=/IcicleLarvaGameplay/Assets/Animation/Flight/OneHandedFlight/IcicleLarva_Flight_Noise_Idle_OneHanded_Loop}
    ) {alphaBoolBlend {blendOption=Linear}})
  ]
 }
]
variables [
 HoverPivotBackToForward {
  type=AnimSequenceBase
  default=/IcicleLarvaGameplay/Assets/Animation/Flight/Hover/Transitions/IcicleLarva_FlightTrans_Hover_Backwards_to_Forward
 }
 HoverPivotForwardToBack {
  type=AnimSequenceBase
  default=&IcicleLarva_FlightTrans_Hover_Forward_to_Backwards
 }
 HoverPivotLeftToRight {type=AnimSequenceBase, default=/IcicleLarvaGameplay/Assets/Animation/Flight/Hover/IcicleLarva_FlightTrans_Hover_Left_to_Right}
 HoverPivotRightToLeft {type=AnimSequenceBase, default=/IcicleLarvaGameplay/Assets/Animation/Flight/Hover/IcicleLarva_FlightTrans_Hover_Right_to_Left}
]
